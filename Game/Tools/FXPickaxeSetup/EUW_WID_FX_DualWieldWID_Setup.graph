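# Editor utility widget for dual-wield pickaxe FX setup, exposing main and offhand idle, swing, trail and impact effects in details views.
# Tick polls for pickaxe or weapon definition changes and toggles the advanced impacts view.
# Weapon changes reload fields from the definition, preferring Niagara over Cascade and flagging missing default impacts, or clear them if invalid.
# The offhand idle socket is written into the main Idle VFX Socket Name.
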EditorUtilityWidgetBlueprint /Game/Tools/FXPickaxeSetup/EUW_WID_FX_DualWieldWID_Setup
graphs [
 EventGraph {nodes [
   Clear Effects Data Info() {is=title K2Node_CustomEvent}
   > n1 = Sequence!(execute) -> (then_0, then_1, then_2:Exec, then_3:Exec, then_4:Exec)
   > Set Idle VFX(Idle VFX:FXSystem Asset Soft Object Reference) -> (Output_Get:FXSystem Asset Soft Object Reference)
   > Set Idle VFX Socket Name(Idle VFX Socket Name:Name=None) -> (Output_Get:Name=None)
   > Set Offhand Idle VFX(Offhand Idle VFX:FXSystem Asset Soft Object Reference) -> (Output_Get:FXSystem Asset Soft Object Reference)
   > Set Offhand Idle VFX Socket Name(Offhand Idle VFX Socket Name:Name=None) -> (Output_Get:Name=None)
   Set Swing VFX(execute=@n1.then_1, Swing VFX:FXSystem Asset Soft Object Reference) -> (Output_Get:FXSystem Asset Soft Object Reference)
   > Set Swing VFX Socket Name(Swing VFX Socket Name:Name=None) -> (Output_Get:Name=None)
   > Set Offhand Swing VFX(Offhand Swing VFX:FXSystem Asset Soft Object Reference) -> (Output_Get:FXSystem Asset Soft Object Reference)
   > Set Offhand Swing VFX Socket Name(Offhand Swing VFX Socket Name:Name=None) -> (Output_Get:Name=None)
   Set AnimTrail VFX(execute=@n1.then_2, AnimTrail VFX:FXSystem Asset Soft Object Reference) -> (Output_Get:FXSystem Asset Soft Object Reference)
   > Set Anim Trails Niagara Skeleton DI Parameter Name(Anim Trails Niagara Skeleton DI Parameter Name:Name=None) -> (Output_Get:Name=None)
   > Set Anim Trails First Socket Name(Anim Trails First Socket Name:Name=None) -> (Output_Get:Name=None)
   > Set Anim Trails Second Socket Name(Anim Trails Second Socket Name:Name=None) -> (Output_Get:Name=None)
   > Set Anim Trails Width(Anim Trails Width:double=0) -> (Output_Get:double)
   > Set Attach Anim Trails to Weapon(Attach Anim Trails to Weapon:bool=false) -> (Output_Get:bool)
   Set Offhand AnimTrail VFX(execute=@n1.then_3, Offhand AnimTrail VFX:FXSystem Asset Soft Object Reference) -> (Output_Get:FXSystem Asset Soft Object Reference)
   > Set Offhand Anim Trails Niagara Skeleton DI Parameter Name(Offhand Anim Trails Niagara Skeleton DI Parameter Name:Name=None) -> (Output_Get:Name=None)
   > Set Offhand Anim Trails First Socket Name(Offhand Anim Trails First Socket Name:Name=None) -> (Output_Get:Name=None)
   > Set Offhand Anim Trails Second Socket Name(Offhand Anim Trails Second Socket Name:Name=None) -> (Output_Get:Name=None)
   > Set Offhand Anim Trails Width(Offhand Anim Trails Width:double=0) -> (Output_Get:double)
   > Set Offhand Attach Anim Trails to Weapon(Offhand Attach Anim Trails to Weapon:bool=false) -> (Output_Get:bool)
   Clear!(
    execute=@n1.then_4
    TargetMap:Map of EPhysicalSurface Enums to Particle System Soft Object References=Get Impact Cascade Physical Surface Effects Map() -> (Impact Cascade Physical Surface Effects Map:Map of EPhysicalSurface Enums to Particle System Soft Object References=SurfaceType_Default)
   ) -> (then) {is=title K2Node_CallFunction}
   > Clear(TargetMap:Map of EPhysicalSurface Enums to Niagara System Soft Object References=Get Impact Niagara Physical Surface Effects Map() -> (Impact Niagara Physical Surface Effects Map:Map of EPhysicalSurface Enums to Niagara System Soft Object References=SurfaceType_Default))
   > Set Impact VFX_Default(Impact VFX_Default:FXSystem Asset Soft Object Reference) -> (Output_Get:FXSystem Asset Soft Object Reference)
   > Clear(TargetMap=Get Offhand Impact Niagara Physical Surface Effects Map() -> (Offhand Impact Niagara Physical Surface Effects Map:Map of EPhysicalSurface Enums to Niagara System Soft Object References=SurfaceType_Default))
   > Clear(TargetMap:Map of EPhysicalSurface Enums to Particle System Soft Object References=Get Offhand Impact Cascade Physical Surface Effects Map() -> (Offhand Impact Cascade Physical Surface Effects Map:Map of EPhysicalSurface Enums to Particle System Soft Object References=SurfaceType_Default))
   > Set Offhand Impact VFX_Default(Offhand Impact VFX_Default:FXSystem Asset Soft Object Reference) -> (Output_Get:FXSystem Asset Soft Object Reference)
   Event PreConstruct()
   > Do Once()
   n2 = Get DetailsView_WID_Advanced() -> (DetailsView_WID_Advanced:Details View Object Reference)
   > SetObject(self=[Get DetailsView_WID() -> (DetailsView_WID:Details View Object Reference), @n2], NewObject=Self-Reference())
   > SetVisibility(self=@n2, InVisibility=Collapsed)
   Event Tick()
   > n3 = Sequence!(execute) -> (then_0, then_1, then_2)
   n4 = Branch(
    execute=@n3.then_2
    Condition="Equal (Object)"(
     A=Get Pickaxe ID() -> (Pickaxe ID:Athena Pickaxe Item Definition Object Reference)
     B=Get Pickaxe ID_Setter() -> (Pickaxe ID_Setter:Athena Pickaxe Item Definition Object Reference)
    )
   )
   > n5 = Do Once(Reset=@n6.then)
   Do Once(execute=@n4.else, Reset=@n5.Completed)
   > n6 = Call Pickaxe Changed!(
    execute:Exec
    self:EUW WID FX Dual Wield WID Setup Object Reference
    PIckaxeID:Athena Pickaxe Item Definition Object Reference=Get Pickaxe ID()
   ) -> (then:Exec) {is=title K2Node_CallDelegate}
   n7 = Branch(execute=@n3.then_0, Condition=Get ShowAdvancedImpactsSetup() -> (ShowAdvancedImpactsSetup:bool))
   > Do Once(Reset=@n9.then)
   > n8 = SetVisibility(self=Get DetailsView_WID_Advanced(), InVisibility=Visible)
   Do Once(execute=@n7.else, Reset=@n8.then, Start Closed=true)
   > n9 = SetVisibility(self=Get DetailsView_WID_Advanced(), InVisibility=Collapsed)
   n10 = Branch(
    execute=@n3.then_1
    Condition="Equal (Object)"(
     A=Get Dual Wield Weapon Item Definition() -> (Dual Wield Weapon Item Definition:Fort Weapon Melee Dual Wield Item Definition Object Reference)
     B=Get Weapon Item Definition_Setter() -> (Weapon Item Definition_Setter:Fort Weapon Melee Item Definition Object Reference)
    )
   )
   > n11 = Do Once(Reset=@n12.then, Start Closed=true)
   Do Once(execute=@n10.else, Reset=@n11.Completed)
   > n12 = Setup Effect Data Sections()
  ]}
 Log Changes {nodes [Log Changes() {is=title K2Node_FunctionEntry}]}
 Pickaxe Changed {nodes [Pickaxe Changed!() -> (then, PIckaxeID:Athena Pickaxe Item Definition Object Reference) {is=title K2Node_FunctionEntry}]}
 Setup Effect Data Sections {nodes [
   Setup Effect Data Sections() {is=title K2Node_FunctionEntry}
   n1 = Get Dual Wield Weapon Item Definition()
   > Set Weapon Item Definition_Setter(Weapon Item Definition_Setter:Fort Weapon Melee Item Definition Object Reference=@n1) -> (Output_Get:Fort Weapon Melee Item Definition Object Reference)
   > n2 = Is Valid(InputObject=@n1)
   Clear Effects Data Info(execute=@n2.Is Not Valid) {is=title K2Node_CallFunction}
   n3 = Sequence!(execute=@n2.Is Valid) -> (then_0, then_1, then_2, then_3:Exec, then_4:Exec, then_5:Exec, then_6:Exec, then_7:Exec, then_8:Exec, then_9:Exec)
   Call Log Changes!(execute:Exec=@n3.then_9, self:EUW WID FX Dual Wield WID Setup Object Reference) -> (then:Exec) {is=title K2Node_CallDelegate}
   n4 = GetSwingEffectDataOffhand!(self:Fort Weapon Melee Dual Wield Item Definition Object Reference=Get Dual Wield Weapon Item Definition()) -> (
    ReturnValue_SwingEffect:Particle System Soft Object Reference
    ReturnValue_SwingEffectNiagara:Niagara System Soft Object Reference
    ReturnValue_SwingFXSocketName:Name
   )
   n5 = Branch(execute=@n3.then_4, Condition=IsValidSoftObjectReference(SoftObjectReference=@n4.ReturnValue_SwingEffectNiagara))
   > Set Offhand Swing VFX(Offhand Swing VFX=@n4.ReturnValue_SwingEffectNiagara)
   > n6 = Set Use Cascade Swing__Offhand(Use Cascade Swing__Offhand:bool=false) -> (Output_Get:bool)
   Set Offhand Swing VFX Socket Name(execute=[@n9.then, @n6.then, @n8.then], Offhand Swing VFX Socket Name=@n4.ReturnValue_SwingFXSocketName) -> (Output_Get=None)
   n7 = Branch(execute=@n5.else, Condition=IsValidSoftObjectReference(SoftObjectReference=@n4))
   > Set Offhand Swing VFX(Offhand Swing VFX=@n4)
   > n8 = Set Use Cascade Swing__Offhand(Use Cascade Swing__Offhand=true)
   Set Offhand Swing VFX(execute=@n7.else)
   > n9 = Set Use Cascade Swing__Offhand(Use Cascade Swing__Offhand=false)
   n10 = Get Dual Wield Weapon Item Definition()
   Set Pickaxe Weapon Actor Class(
    execute=@n3.then_0
    Pickaxe Weapon Actor Class:Fort Weapon Soft Class Reference=Make Soft Reference!(Input:Fort Weapon Class Reference=GetWeaponActorClass!(self:Fort Weapon Item Definition Object Reference=@n10) -> (ReturnValue:Fort Weapon Class Reference)) -> (Output:Fort Weapon Soft Class Reference)
   ) -> (Output_Get:Fort Weapon Soft Class Reference)
   n11 = GetIdleEffectData!(self:Fort Weapon Melee Item Definition Object Reference=Get Dual Wield Weapon Item Definition()) -> (
    ReturnValue_IdleEffect:Particle System Soft Object Reference
    ReturnValue_IdleEffectNiagara:Niagara System Soft Object Reference
    ReturnValue_IdleFXSocketName:Name
   )
   n12 = Branch(execute=@n3.then_1, Condition=IsValidSoftObjectReference(SoftObjectReference=@n11.ReturnValue_IdleEffectNiagara))
   > Set Idle VFX(Idle VFX=@n11.ReturnValue_IdleEffectNiagara)
   > n13 = Set Use Cascade Idle(Use Cascade Idle:bool=false) -> (Output_Get:bool)
   Set Idle VFX Socket Name(execute=[@n15.then, @n16.then, @n13.then], Idle VFX Socket Name=@n11.ReturnValue_IdleFXSocketName) -> (Output_Get=None)
   n14 = Branch(execute=@n12.else, Condition=IsValidSoftObjectReference(SoftObjectReference=@n11))
   > Set Idle VFX(Idle VFX=@n11)
   > n15 = Set Use Cascade Idle(Use Cascade Idle=true)
   Set Idle VFX(execute=@n14.else)
   > n16 = Set Use Cascade Idle(Use Cascade Idle=false)
   n17 = GetAnimTrailDataOffhand!(self:Fort Weapon Melee Dual Wield Item Definition Object Reference=Get Dual Wield Weapon Item Definition()) -> (
    ReturnValue_AnimTrails:Particle System Soft Object Reference
    ReturnValue_AnimTrailsNiagara:Niagara System Soft Object Reference
    ReturnValue_NiagaraSkeletonDIVariableName:Name
    ReturnValue_AnimTrailsFirstSocketName:Name
    ReturnValue_AnimTrailsSecondSocketName:Name
    ReturnValue_AnimTrailsWidth:float=~
    ReturnValue_bAttachAnimTrailsToWeapon:bool=~
   )
   n18 = Branch(execute=@n3.then_6, Condition=IsValidSoftObjectReference(SoftObjectReference=@n17.ReturnValue_AnimTrailsNiagara))
   > Set Offhand AnimTrail VFX(Offhand AnimTrail VFX=@n17.ReturnValue_AnimTrailsNiagara)
   > n19 = Set Use Cascade AnimTrail__Offhand(Use Cascade AnimTrail__Offhand:bool=false) -> (Output_Get:bool)
   Set Offhand Anim Trails Niagara Skeleton DI Parameter Name(
    execute=[@n21.then, @n22.then, @n19.then]
    Offhand Anim Trails Niagara Skeleton DI Parameter Name=@n17.ReturnValue_NiagaraSkeletonDIVariableName
   ) -> (Output_Get=None)
   > Set Offhand Anim Trails First Socket Name(Offhand Anim Trails First Socket Name=@n17.ReturnValue_AnimTrailsFirstSocketName) -> (Output_Get=None)
   > Set Offhand Anim Trails Second Socket Name(Offhand Anim Trails Second Socket Name=@n17.ReturnValue_AnimTrailsSecondSocketName) -> (Output_Get=None)
   > Set Offhand Anim Trails Width(Offhand Anim Trails Width=@n17.ReturnValue_AnimTrailsWidth)
   > Set Offhand Attach Anim Trails to Weapon(Offhand Attach Anim Trails to Weapon=@n17.ReturnValue_bAttachAnimTrailsToWeapon)
   n20 = Branch(execute=@n18.else, Condition=IsValidSoftObjectReference(SoftObjectReference=@n17))
   > Set Offhand AnimTrail VFX(Offhand AnimTrail VFX=@n17)
   > n21 = Set Use Cascade AnimTrail__Offhand(Use Cascade AnimTrail__Offhand=true)
   Set Offhand AnimTrail VFX(execute=@n20.else)
   > n22 = Set Use Cascade AnimTrail__Offhand(Use Cascade AnimTrail__Offhand=false)
   n23 = GetImpactEffectDataOffhand!(self:Fort Weapon Melee Dual Wield Item Definition Object Reference=Get Dual Wield Weapon Item Definition()) -> (
    ReturnValue_ImpactPhysicalSurfaceEffectsMap:Map of EPhysicalSurface Enums to Particle System Soft Object References
    ReturnValue_ImpactNiagaraPhysicalSurfaceEffectsMap:Map of EPhysicalSurface Enums to Niagara System Soft Object References
   )
   Set Offhand Impact Cascade Physical Surface Effects Map(
    execute=@n3.then_8
    Offhand Impact Cascade Physical Surface Effects Map:Map of EPhysicalSurface Enums to Particle System Soft Object References=@n23
   ) -> (Output_Get:Map of EPhysicalSurface Enums to Particle System Soft Object References=SurfaceType_Default)
   > Set Offhand Impact Niagara Physical Surface Effects Map(Offhand Impact Niagara Physical Surface Effects Map:Map of EPhysicalSurface Enums to Niagara System Soft Object References=@n23.ReturnValue_ImpactNiagaraPhysicalSurfaceEffectsMap) -> (Output_Get:Map of EPhysicalSurface Enums to Niagara System Soft Object References=SurfaceType_Default)
   n24 = Find(
    TargetMap:Map of EPhysicalSurface Enums to Niagara System Soft Object References=@n23.ReturnValue_ImpactNiagaraPhysicalSurfaceEffectsMap
    Key:"EPhysicalSurface Enum (by ref)"=SurfaceType_Default
   ) -> (Value:Niagara System Soft Object Reference)
   > n25 = Branch(Condition=@n24.ReturnValue)
   > Set Offhand Impact VFX_Default(Offhand Impact VFX_Default=@n24)
   > Set Use Cascade Impacts__Offhand(Use Cascade Impacts__Offhand:bool=false) -> (Output_Get:bool)
   > Set Missing Impacts__Offhand(Missing Impacts__Offhand:bool=false) -> (Output_Get:bool)
   n26 = Find(TargetMap:Map of EPhysicalSurface Enums to Particle System Soft Object References=@n23, Key=SurfaceType_Default) -> (Value:Particle System Soft Object Reference)
   n27 = Branch(execute=@n25.else, Condition=@n26.ReturnValue)
   > Set Offhand Impact VFX_Default(Offhand Impact VFX_Default=@n26)
   > Set Use Cascade Impacts__Offhand(Use Cascade Impacts__Offhand=true)
   > Set Missing Impacts__Offhand(Missing Impacts__Offhand=false)
   Set Offhand Impact VFX_Default(execute=@n27.else)
   > Set Use Cascade Impacts__Offhand(Use Cascade Impacts__Offhand=false)
   > Set Missing Impacts__Offhand(Missing Impacts__Offhand=true)
   n28 = GetSwingEffectData!(self:Fort Weapon Melee Item Definition Object Reference=Get Dual Wield Weapon Item Definition()) -> (
    ReturnValue_SwingEffect:Particle System Soft Object Reference
    ReturnValue_SwingEffectNiagara:Niagara System Soft Object Reference
    ReturnValue_SwingFXSocketName:Name
   )
   n29 = Branch(execute=@n3.then_3, Condition=IsValidSoftObjectReference(SoftObjectReference=@n28.ReturnValue_SwingEffectNiagara))
   > Set Swing VFX(Swing VFX=@n28.ReturnValue_SwingEffectNiagara)
   > n30 = Set Use Cascade Swing(Use Cascade Swing:bool=false) -> (Output_Get:bool)
   Set Swing VFX Socket Name(execute=[@n32.then, @n33.then, @n30.then], Swing VFX Socket Name=@n28.ReturnValue_SwingFXSocketName) -> (Output_Get=None)
   n31 = Branch(execute=@n29.else, Condition=IsValidSoftObjectReference(SoftObjectReference=@n28))
   > Set Swing VFX(Swing VFX=@n28)
   > n32 = Set Use Cascade Swing(Use Cascade Swing=true)
   Set Swing VFX(execute=@n31.else)
   > n33 = Set Use Cascade Swing(Use Cascade Swing=false)
   n34 = GetAnimTrailData!(self:Fort Weapon Melee Item Definition Object Reference=Get Dual Wield Weapon Item Definition()) -> (
    ReturnValue_AnimTrails:Particle System Soft Object Reference
    ReturnValue_AnimTrailsNiagara:Niagara System Soft Object Reference
    ReturnValue_NiagaraSkeletonDIVariableName:Name
    ReturnValue_AnimTrailsFirstSocketName:Name
    ReturnValue_AnimTrailsSecondSocketName:Name
    ReturnValue_AnimTrailsWidth:float=~
    ReturnValue_bAttachAnimTrailsToWeapon:bool=~
   )
   n35 = Branch(execute=@n3.then_5, Condition=IsValidSoftObjectReference(SoftObjectReference=@n34.ReturnValue_AnimTrailsNiagara))
   > Set AnimTrail VFX(AnimTrail VFX=@n34.ReturnValue_AnimTrailsNiagara)
   > n36 = Set Use Cascade AnimTrail(Use Cascade AnimTrail:bool=false) -> (Output_Get:bool)
   Set Anim Trails Niagara Skeleton DI Parameter Name(execute=[@n38.then, @n39.then, @n36.then], Anim Trails Niagara Skeleton DI Parameter Name=@n34.ReturnValue_NiagaraSkeletonDIVariableName) -> (Output_Get=None)
   > Set Anim Trails First Socket Name(Anim Trails First Socket Name=@n34.ReturnValue_AnimTrailsFirstSocketName) -> (Output_Get=None)
   > Set Anim Trails Second Socket Name(Anim Trails Second Socket Name=@n34.ReturnValue_AnimTrailsSecondSocketName) -> (Output_Get=None)
   > Set Anim Trails Width(Anim Trails Width=@n34.ReturnValue_AnimTrailsWidth)
   > Set Attach Anim Trails to Weapon(Attach Anim Trails to Weapon=@n34.ReturnValue_bAttachAnimTrailsToWeapon)
   n37 = Branch(execute=@n35.else, Condition=IsValidSoftObjectReference(SoftObjectReference=@n34))
   > Set AnimTrail VFX(AnimTrail VFX=@n34)
   > n38 = Set Use Cascade AnimTrail(Use Cascade AnimTrail=true)
   Set AnimTrail VFX(execute=@n37.else)
   > n39 = Set Use Cascade AnimTrail(Use Cascade AnimTrail=false)
   n40 = GetImpactEffectData!(self:Fort Weapon Melee Item Definition Object Reference=Get Dual Wield Weapon Item Definition()) -> (
    ReturnValue_ImpactPhysicalSurfaceEffectsMap:Map of EPhysicalSurface Enums to Particle System Soft Object References
    ReturnValue_ImpactNiagaraPhysicalSurfaceEffectsMap:Map of EPhysicalSurface Enums to Niagara System Soft Object References
   )
   Set Impact Cascade Physical Surface Effects Map(execute=@n3.then_7, Impact Cascade Physical Surface Effects Map:Map of EPhysicalSurface Enums to Particle System Soft Object References=@n40) -> (Output_Get:Map of EPhysicalSurface Enums to Particle System Soft Object References=SurfaceType_Default)
   > Set Impact Niagara Physical Surface Effects Map(Impact Niagara Physical Surface Effects Map:Map of EPhysicalSurface Enums to Niagara System Soft Object References=@n40.ReturnValue_ImpactNiagaraPhysicalSurfaceEffectsMap) -> (Output_Get:Map of EPhysicalSurface Enums to Niagara System Soft Object References=SurfaceType_Default)
   n41 = Find(
    TargetMap:Map of EPhysicalSurface Enums to Niagara System Soft Object References=@n40.ReturnValue_ImpactNiagaraPhysicalSurfaceEffectsMap
    Key=SurfaceType_Default
   ) -> (Value:Niagara System Soft Object Reference)
   > n42 = Branch(Condition=@n41.ReturnValue)
   > Set Impact VFX_Default(Impact VFX_Default=@n41)
   > Set Use Cascade Impacts(Use Cascade Impacts:bool=false) -> (Output_Get:bool)
   > Set Missing Impacts(Missing Impacts:bool=false) -> (Output_Get:bool)
   n43 = Find(TargetMap:Map of EPhysicalSurface Enums to Particle System Soft Object References=@n40, Key=SurfaceType_Default) -> (Value:Particle System Soft Object Reference)
   n44 = Branch(execute=@n42.else, Condition=@n43.ReturnValue)
   > Set Impact VFX_Default(Impact VFX_Default=@n43)
   > Set Use Cascade Impacts(Use Cascade Impacts=true)
   > Set Missing Impacts(Missing Impacts=false)
   Set Impact VFX_Default(execute=@n44.else)
   > Set Use Cascade Impacts(Use Cascade Impacts=false)
   > Set Missing Impacts(Missing Impacts=true)
   n45 = GetIdleEffectDataOffhand!(self:Fort Weapon Melee Dual Wield Item Definition Object Reference=Get Dual Wield Weapon Item Definition()) -> (
    ReturnValue_IdleEffect:Particle System Soft Object Reference
    ReturnValue_IdleEffectNiagara:Niagara System Soft Object Reference
    ReturnValue_IdleFXSocketName:Name
   )
   n46 = Branch(execute=@n3.then_2, Condition=IsValidSoftObjectReference(SoftObjectReference=@n45.ReturnValue_IdleEffectNiagara))
   > Set Offhand Idle VFX(Offhand Idle VFX=@n45.ReturnValue_IdleEffectNiagara)
   > n47 = Set Use Cascade Idle_Offhand(Use Cascade Idle_Offhand:bool=false) -> (Output_Get:bool)
   Set Idle VFX Socket Name(execute=[@n50.then, @n47.then, @n49.then], Idle VFX Socket Name=@n45.ReturnValue_IdleFXSocketName) -> (Output_Get=None)
   n48 = Branch(execute=@n46.else, Condition=IsValidSoftObjectReference(SoftObjectReference=@n45))
   > Set Offhand Idle VFX(Offhand Idle VFX=@n45)
   > n49 = Set Use Cascade Idle_Offhand(Use Cascade Idle_Offhand=true)
   Set Offhand Idle VFX(execute=@n48.else)
   > n50 = Set Use Cascade Idle_Offhand(Use Cascade Idle_Offhand=false)
   Get WID Widget() -> (WID Widget:EUW WID FX WID Setup Object Reference)
  ]}
]
variables [
 Anim Trails First Socket Name {type=string, default=None}
 Anim Trails Niagara Skeleton DI Parameter Name {type=string, default=None}
 Anim Trails Second Socket Name {type=string, default=None}
 Anim Trails Width {type=number}
 AnimTrail VFX {type=FXSystemAsset, default=null}
 Attach Anim Trails to Weapon {type=boolean}
 Dual Wield Weapon Item Definition {type=FortWeaponMeleeDualWieldItemDefinition, default=null}
 Idle VFX {type=FXSystemAsset, default=null}
 Idle VFX Socket Name {type=string, default=None}
 Impact Cascade Physical Surface Effects Map {default {}}
 Impact Niagara Physical Surface Effects Map {default {}}
 Impact VFX_Default {type=FXSystemAsset, default=null}
 Log Changes {default=()}
 Missing Impacts {type=boolean}
 Missing Impacts__Offhand {type=boolean}
 Offhand Anim Trails First Socket Name {type=string, default=None}
 Offhand Anim Trails Niagara Skeleton DI Parameter Name {type=string, default=None}
 Offhand Anim Trails Second Socket Name {type=string, default=None}
 Offhand Anim Trails Width {type=number}
 Offhand AnimTrail VFX {type=FXSystemAsset, default=null}
 Offhand Attach Anim Trails to Weapon {type=boolean}
 Offhand Idle VFX {type=FXSystemAsset, default=null}
 Offhand Idle VFX Socket Name {type=string, default=None}
 Offhand Impact Cascade Physical Surface Effects Map {default {}}
 Offhand Impact Niagara Physical Surface Effects Map {default {}}
 Offhand Impact VFX_Default {type=FXSystemAsset, default=null}
 Offhand Swing VFX {type=FXSystemAsset, default=null}
 Offhand Swing VFX Socket Name {type=string, default=None}
 Pickaxe Changed {default=()}
 Pickaxe ID {type=AthenaPickaxeItemDefinition, default=null}
 Pickaxe ID_Setter {type=AthenaPickaxeItemDefinition, default=null}
 Pickaxe Weapon Actor Class {type=Class, default=null}
 ShowAdvancedImpactsSetup {type=boolean}
 Swing VFX {type=FXSystemAsset, default=null}
 Swing VFX Socket Name {type=string, default=None}
 Use Cascade AnimTrail {type=boolean}
 Use Cascade AnimTrail__Offhand {type=boolean}
 Use Cascade Idle {type=boolean}
 Use Cascade Idle_Offhand {type=boolean}
 Use Cascade Impacts {type=boolean}
 Use Cascade Impacts__Offhand {type=boolean}
 Use Cascade Swing {type=boolean}
 Use Cascade Swing__Offhand {type=boolean}
 Weapon Item Definition_Setter {type=FortWeaponMeleeItemDefinition, default=null}
]
widgets {
 info {rootWidgetClass=&StackBox, widgetCount=5}
 widgets [
  StackBox_0 {widgetClassPath=&StackBox, widget=StackBox_0}
  CommonHierarchicalScrollBox_73 {
   widgetClassPath=/Script/CommonUI.CommonHierarchicalScrollBox
   parent=StackBox_0
   widget=CommonHierarchicalScrollBox_73
   properties {
    widgetBarStyle {
     thickness=8
     draggedThumbImage {
      bIsDynamicallyLoaded=false
      drawAs=RoundedBox
      imageType=NoImage
      mirroring=NoMirror
      resourceName=None
      resourceObject=null
      tiling=NoTile
      imageSize=<0 0>
      margin=<ltrb 0 0 0 0>
      outlineSettings {
       bUseBrushTransparency=false
       roundingType=FixedRadius
       width=0
       color {colorUseRule=UseColor_Specified, specifiedColor=<rgba 0 0 0 0>}
       cornerRadii=<4 4 4 4>
      }
      tintColor {colorUseRule=UseColor_Specified, specifiedColor=<rgba 0.2159 0.2159 0.2159 1>}
      uVRegion {bIsValid=false, max=<0 0>, min=<0 0>}
     }
     horizontalBackgroundImage {
      bIsDynamicallyLoaded=false
      drawAs=NoDrawType
      imageType=NoImage
      mirroring=NoMirror
      resourceName=None
      resourceObject=null
      tiling=NoTile
      imageSize=<0 0>
      margin=<ltrb 0 0 0 0>
      outlineSettings {
       bUseBrushTransparency=false
       roundingType=HalfHeightRadius
       width=0
       color {colorUseRule=UseColor_Specified, specifiedColor=<rgba 0 0 0 0>}
       cornerRadii=<0 0 0 0>
      }
      tintColor {colorUseRule=UseColor_Specified, specifiedColor=<rgba 1 1 1 1>}
      uVRegion {bIsValid=false, max=<0 0>, min=<0 0>}
     }
     horizontalBottomSlotImage {
      bIsDynamicallyLoaded=false
      drawAs=NoDrawType
      imageType=NoImage
      mirroring=NoMirror
      resourceName=None
      resourceObject=null
      tiling=NoTile
      imageSize=<0 0>
      margin=<ltrb 0 0 0 0>
      outlineSettings {
       bUseBrushTransparency=false
       roundingType=HalfHeightRadius
       width=0
       color {colorUseRule=UseColor_Specified, specifiedColor=<rgba 0 0 0 0>}
       cornerRadii=<0 0 0 0>
      }
      tintColor {colorUseRule=UseColor_Specified, specifiedColor=<rgba 1 1 1 1>}
      uVRegion {bIsValid=false, max=<0 0>, min=<0 0>}
     }
     horizontalTopSlotImage {
      bIsDynamicallyLoaded=false
      drawAs=NoDrawType
      imageType=NoImage
      mirroring=NoMirror
      resourceName=None
      resourceObject=null
      tiling=NoTile
      imageSize=<0 0>
      margin=<ltrb 0 0 0 0>
      outlineSettings {
       bUseBrushTransparency=false
       roundingType=HalfHeightRadius
       width=0
       color {colorUseRule=UseColor_Specified, specifiedColor=<rgba 0 0 0 0>}
       cornerRadii=<0 0 0 0>
      }
      tintColor {colorUseRule=UseColor_Specified, specifiedColor=<rgba 1 1 1 1>}
      uVRegion {bIsValid=false, max=<0 0>, min=<0 0>}
     }
     hoveredThumbImage {
      bIsDynamicallyLoaded=false
      drawAs=RoundedBox
      imageType=NoImage
      mirroring=NoMirror
      resourceName=None
      resourceObject=null
      tiling=NoTile
      imageSize=<0 0>
      margin=<ltrb 0 0 0 0>
      outlineSettings {
       bUseBrushTransparency=false
       roundingType=FixedRadius
       width=0
       color {colorUseRule=UseColor_Specified, specifiedColor=<rgba 0 0 0 0>}
       cornerRadii=<4 4 4 4>
      }
      tintColor {colorUseRule=UseColor_Specified, specifiedColor=<rgba 0.2159 0.2159 0.2159 1>}
      uVRegion {bIsValid=false, max=<0 0>, min=<0 0>}
     }
     normalThumbImage {
      bIsDynamicallyLoaded=false
      drawAs=RoundedBox
      imageType=NoImage
      mirroring=NoMirror
      resourceName=None
      resourceObject=null
      tiling=NoTile
      imageSize=<0 0>
      margin=<ltrb 0 0 0 0>
      outlineSettings {
       bUseBrushTransparency=false
       roundingType=FixedRadius
       width=0
       color {colorUseRule=UseColor_Specified, specifiedColor=<rgba 0 0 0 0>}
       cornerRadii=<4 4 4 4>
      }
      tintColor {colorUseRule=UseColor_Specified, specifiedColor=<rgba 0.0953 0.0953 0.0953 1>}
      uVRegion {bIsValid=false, max=<0 0>, min=<0 0>}
     }
     verticalBackgroundImage {
      bIsDynamicallyLoaded=false
      drawAs=NoDrawType
      imageType=NoImage
      mirroring=NoMirror
      resourceName=None
      resourceObject=null
      tiling=NoTile
      imageSize=<0 0>
      margin=<ltrb 0 0 0 0>
      outlineSettings {
       bUseBrushTransparency=false
       roundingType=HalfHeightRadius
       width=0
       color {colorUseRule=UseColor_Specified, specifiedColor=<rgba 0 0 0 0>}
       cornerRadii=<0 0 0 0>
      }
      tintColor {colorUseRule=UseColor_Specified, specifiedColor=<rgba 1 1 1 1>}
      uVRegion {bIsValid=false, max=<0 0>, min=<0 0>}
     }
     verticalBottomSlotImage {
      bIsDynamicallyLoaded=false
      drawAs=NoDrawType
      imageType=NoImage
      mirroring=NoMirror
      resourceName=None
      resourceObject=null
      tiling=NoTile
      imageSize=<0 0>
      margin=<ltrb 0 0 0 0>
      outlineSettings {
       bUseBrushTransparency=false
       roundingType=HalfHeightRadius
       width=0
       color {colorUseRule=UseColor_Specified, specifiedColor=<rgba 0 0 0 0>}
       cornerRadii=<0 0 0 0>
      }
      tintColor {colorUseRule=UseColor_Specified, specifiedColor=<rgba 1 1 1 1>}
      uVRegion {bIsValid=false, max=<0 0>, min=<0 0>}
     }
     verticalTopSlotImage {
      bIsDynamicallyLoaded=false
      drawAs=NoDrawType
      imageType=NoImage
      mirroring=NoMirror
      resourceName=None
      resourceObject=null
      tiling=NoTile
      imageSize=<0 0>
      margin=<ltrb 0 0 0 0>
      outlineSettings {
       bUseBrushTransparency=false
       roundingType=HalfHeightRadius
       width=0
       color {colorUseRule=UseColor_Specified, specifiedColor=<rgba 0 0 0 0>}
       cornerRadii=<0 0 0 0>
      }
      tintColor {colorUseRule=UseColor_Specified, specifiedColor=<rgba 1 1 1 1>}
      uVRegion {bIsValid=false, max=<0 0>, min=<0 0>}
     }
    }
    widgetStyle {
     barThickness=8
     bottomShadowBrush {
      bIsDynamicallyLoaded=false
      drawAs=Box
      imageType=FullColor
      mirroring=NoMirror
      resourceName=../../../Engine/Content/Slate/Common/ScrollBoxShadowBottom.png
      resourceObject=null
      tiling=NoTile
      imageSize=<16 8>
      margin=<ltrb 0.5 0 0.5 1>
      outlineSettings {
       bUseBrushTransparency=false
       roundingType=HalfHeightRadius
       width=0
       color {colorUseRule=UseColor_Specified, specifiedColor=<rgba 0 0 0 0>}
       cornerRadii=<0 0 0 0>
      }
      tintColor {colorUseRule=UseColor_Specified, specifiedColor=<rgba 1 1 1 1>}
      uVRegion {bIsValid=false, max=<0 0>, min=<0 0>}
     }
     horizontalScrolledContentPadding=<ltrb 0 0 1 0>
     leftShadowBrush {
      bIsDynamicallyLoaded=false
      drawAs=Box
      imageType=FullColor
      mirroring=NoMirror
      resourceName=../../../Engine/Content/Slate/Common/ScrollBoxShadowLeft.png
      resourceObject=null
      tiling=NoTile
      imageSize=<8 16>
      margin=<ltrb 1 0.5 0 0.5>
      outlineSettings {
       bUseBrushTransparency=false
       roundingType=HalfHeightRadius
       width=0
       color {colorUseRule=UseColor_Specified, specifiedColor=<rgba 0 0 0 0>}
       cornerRadii=<0 0 0 0>
      }
      tintColor {colorUseRule=UseColor_Specified, specifiedColor=<rgba 1 1 1 1>}
      uVRegion {bIsValid=false, max=<0 0>, min=<0 0>}
     }
     rightShadowBrush {
      bIsDynamicallyLoaded=false
      drawAs=Box
      imageType=FullColor
      mirroring=NoMirror
      resourceName=../../../Engine/Content/Slate/Common/ScrollBoxShadowRight.png
      resourceObject=null
      tiling=NoTile
      imageSize=<8 16>
      margin=<ltrb 0 0.5 1 0.5>
      outlineSettings {
       bUseBrushTransparency=false
       roundingType=HalfHeightRadius
       width=0
       color {colorUseRule=UseColor_Specified, specifiedColor=<rgba 0 0 0 0>}
       cornerRadii=<0 0 0 0>
      }
      tintColor {colorUseRule=UseColor_Specified, specifiedColor=<rgba 1 1 1 1>}
      uVRegion {bIsValid=false, max=<0 0>, min=<0 0>}
     }
     topShadowBrush {
      bIsDynamicallyLoaded=false
      drawAs=Box
      imageType=FullColor
      mirroring=NoMirror
      resourceName=../../../Engine/Content/Slate/Common/ScrollBoxShadowTop.png
      resourceObject=null
      tiling=NoTile
      imageSize=<16 8>
      margin=<ltrb 0.5 1 0.5 0>
      outlineSettings {
       bUseBrushTransparency=false
       roundingType=HalfHeightRadius
       width=0
       color {colorUseRule=UseColor_Specified, specifiedColor=<rgba 0 0 0 0>}
       cornerRadii=<0 0 0 0>
      }
      tintColor {colorUseRule=UseColor_Specified, specifiedColor=<rgba 1 1 1 1>}
      uVRegion {bIsValid=false, max=<0 0>, min=<0 0>}
     }
     verticalScrolledContentPadding=<ltrb 0 0 0 1>
    }
   }
   slot {class=StackBoxSlot, properties {size {sizeRule=Fill, value=1}}}
  }
  DetailsView_WID {
   widgetClassPath=&DetailsView
   parent=CommonHierarchicalScrollBox_73
   widget=DetailsView_WID
   properties {bAllowFiltering=true, bAutoLoadAsset=false, bShowModifiedPropertiesOption=true, categoriesToShow [Setup]}
   slot {class=ScrollBoxSlot, properties {verticalAlignment=VAlign_Top}}
  }
  Spacer {
   widgetClassPath=&Spacer
   parent=CommonHierarchicalScrollBox_73
   widget=Spacer
   properties {size=<5 5>}
   slot {class=ScrollBoxSlot}
  }
  DetailsView_WID_Advanced {
   widgetClassPath=&DetailsView
   parent=CommonHierarchicalScrollBox_73
   widget=DetailsView_WID_Advanced
   properties {bAllowFiltering=true, bAutoLoadAsset=false, bShowModifiedPropertiesOption=true, categoriesToShow [Advanced_Setup]}
   slot {class=ScrollBoxSlot, properties {verticalAlignment=VAlign_Top}}
  }
 ]
}
