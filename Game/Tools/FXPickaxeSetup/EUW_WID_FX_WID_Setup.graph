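# Editor Utility Widget for pickaxe FX setup, exposing its own variables through Setup and Advanced_Setup Details Views, the latter toggled by a bool.
# Tick polls Pickaxe ID and Weapon Item Definition against shadow "_Setter" copies, using DoOnce gates to broadcast Pickaxe Changed or reload effect data.
# Reloading copies the weapon's idle, swing, trail and impact data, preferring Niagara over Cascade and flagging missing default impacts; an invalid weapon clears the effect fields.
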EditorUtilityWidgetBlueprint /Game/Tools/FXPickaxeSetup/EUW_WID_FX_WID_Setup
graphs [
 EventGraph {nodes [
   Clear Effects Data Info() {is=title K2Node_CustomEvent}
   > n1 = Sequence!(execute) -> (then_0, then_1, then_2:Exec, then_3:Exec)
   > Set Idle VFX(Idle VFX:FXSystem Asset Soft Object Reference) -> (Output_Get:FXSystem Asset Soft Object Reference)
   > Set Idle VFX Socket Name(Idle VFX Socket Name:Name=None) -> (Output_Get:Name=None)
   Set Swing VFX(execute=@n1.then_1, Swing VFX:FXSystem Asset Soft Object Reference) -> (Output_Get:FXSystem Asset Soft Object Reference)
   > Set Swing VFX Socket Name(Swing VFX Socket Name:Name=None) -> (Output_Get:Name=None)
   Set AnimTrail VFX(execute=@n1.then_2, AnimTrail VFX:FXSystem Asset Soft Object Reference) -> (Output_Get:FXSystem Asset Soft Object Reference)
   > Set Anim Trails Niagara Skeleton DI Parameter Name(Anim Trails Niagara Skeleton DI Parameter Name:Name=None) -> (Output_Get:Name=None)
   > Set Anim Trails First Socket Name(Anim Trails First Socket Name:Name=None) -> (Output_Get:Name=None)
   > Set Anim Trails Second Socket Name(Anim Trails Second Socket Name:Name=None) -> (Output_Get:Name=None)
   > Set Anim Trails Width(Anim Trails Width:double=0) -> (Output_Get:double)
   > Set Attach Anim Trails to Weapon(Attach Anim Trails to Weapon:bool=false) -> (Output_Get:bool)
   Clear!(
    execute=@n1.then_3
    TargetMap:Map of EPhysicalSurface Enums to Particle System Soft Object References=Get Impact Cascade Physical Surface Effects Map() -> (Impact Cascade Physical Surface Effects Map:Map of EPhysicalSurface Enums to Particle System Soft Object References=SurfaceType_Default)
   ) -> (then) {is=title K2Node_CallFunction}
   > Clear(TargetMap:Map of EPhysicalSurface Enums to Niagara System Soft Object References=Get Impact Niagara Physical Surface Effects Map() -> (Impact Niagara Physical Surface Effects Map:Map of EPhysicalSurface Enums to Niagara System Soft Object References=SurfaceType_Default))
   > Set Impact VFX_Default(Impact VFX_Default:FXSystem Asset Soft Object Reference) -> (Output_Get:FXSystem Asset Soft Object Reference)
   Event PreConstruct()
   > Do Once()
   n2 = Get DetailsView_WID_Advanced() -> (DetailsView_WID_Advanced:Details View Object Reference)
   > SetObject(self=[Get DetailsView_WID() -> (DetailsView_WID:Details View Object Reference), @n2], NewObject=Self-Reference())
   > SetVisibility(self=@n2, InVisibility=Collapsed)
   Event Tick()
   > n3 = Sequence!(execute) -> (then_0, then_1, then_2)
   n4 = Branch(
    execute=@n3.then_2
    Condition="Equal (Object)"(
     A=Get Pickaxe ID() -> (Pickaxe ID:Athena Pickaxe Item Definition Object Reference)
     B=Get Pickaxe ID_Setter() -> (Pickaxe ID_Setter:Athena Pickaxe Item Definition Object Reference)
    )
   )
   > n5 = Do Once(Reset=@n6.then)
   Do Once(execute=@n4.else, Reset=@n5.Completed)
   > n6 = Call Pickaxe Changed!(execute:Exec, self:EUW WID FX WID Setup Object Reference, PIckaxeID:Athena Pickaxe Item Definition Object Reference=Get Pickaxe ID()) -> (then:Exec) {is=title K2Node_CallDelegate}
   n7 = Branch(execute=@n3.then_0, Condition=Get ShowAdvancedImpactsSetup() -> (ShowAdvancedImpactsSetup:bool))
   > Do Once(Reset=@n9.then)
   > n8 = SetVisibility(self=Get DetailsView_WID_Advanced(), InVisibility=Visible)
   Do Once(execute=@n7.else, Reset=@n8.then, Start Closed=true)
   > n9 = SetVisibility(self=Get DetailsView_WID_Advanced(), InVisibility=Collapsed)
   n10 = Branch(
    execute=@n3.then_1
    Condition="Equal (Object)"(
     A=Get Weapon Item Definition() -> (Weapon Item Definition:Fort Weapon Melee Item Definition Object Reference)
     B=Get Weapon Item Definition_Setter() -> (Weapon Item Definition_Setter:Fort Weapon Melee Item Definition Object Reference)
    )
   )
   > n11 = Do Once(Reset=@n12.then, Start Closed=true)
   Do Once(execute=@n10.else, Reset=@n11.Completed)
   > n12 = Setup Effect Data Sections()
  ]}
 Log Changes {nodes [Log Changes() {is=title K2Node_FunctionEntry}]}
 Pickaxe Changed {nodes [Pickaxe Changed!() -> (then, PIckaxeID:Athena Pickaxe Item Definition Object Reference) {is=title K2Node_FunctionEntry}]}
 Setup Effect Data Sections {nodes [
   Setup Effect Data Sections() {is=title K2Node_FunctionEntry}
   n1 = Get Weapon Item Definition()
   > Set Weapon Item Definition_Setter(Weapon Item Definition_Setter:Fort Weapon Melee Item Definition Object Reference=@n1) -> (Output_Get:Fort Weapon Melee Item Definition Object Reference)
   > n2 = Is Valid(InputObject=@n1)
   Clear Effects Data Info(execute=@n2.Is Not Valid) {is=title K2Node_CallFunction}
   n3 = Sequence!(execute=@n2.Is Valid) -> (then_0, then_1, then_2, then_3:Exec, then_4:Exec, then_5:Exec)
   Call Log Changes!(execute:Exec=@n3.then_5, self:EUW WID FX WID Setup Object Reference) -> (then:Exec) {is=title K2Node_CallDelegate}
   n4 = GetSwingEffectData!(self:Fort Weapon Melee Item Definition Object Reference=Get Weapon Item Definition()) -> (
    ReturnValue_SwingEffect:Particle System Soft Object Reference
    ReturnValue_SwingEffectNiagara:Niagara System Soft Object Reference
    ReturnValue_SwingFXSocketName:Name
   )
   n5 = Branch(execute=@n3.then_2, Condition=IsValidSoftObjectReference(SoftObjectReference=@n4.ReturnValue_SwingEffectNiagara))
   > Set Swing VFX(Swing VFX=@n4.ReturnValue_SwingEffectNiagara)
   > n6 = Set Use Cascade Swing(Use Cascade Swing:bool=false) -> (Output_Get:bool)
   Set Swing VFX Socket Name(execute=[@n8.then, @n9.then, @n6.then], Swing VFX Socket Name=@n4.ReturnValue_SwingFXSocketName) -> (Output_Get=None)
   n7 = Branch(execute=@n5.else, Condition=IsValidSoftObjectReference(SoftObjectReference=@n4))
   > Set Swing VFX(Swing VFX=@n4)
   > n8 = Set Use Cascade Swing(Use Cascade Swing=true)
   Set Swing VFX(execute=@n7.else)
   > n9 = Set Use Cascade Swing(Use Cascade Swing=false)
   n10 = Get Weapon Item Definition()
   Set Pickaxe Weapon Actor Class(
    execute=@n3.then_0
    Pickaxe Weapon Actor Class:Fort Weapon Soft Class Reference=Make Soft Reference!(Input:Fort Weapon Class Reference=GetWeaponActorClass!(self:Fort Weapon Item Definition Object Reference=@n10) -> (ReturnValue:Fort Weapon Class Reference)) -> (Output:Fort Weapon Soft Class Reference)
   ) -> (Output_Get:Fort Weapon Soft Class Reference)
   n11 = GetAnimTrailData!(self:Fort Weapon Melee Item Definition Object Reference=Get Weapon Item Definition()) -> (
    ReturnValue_AnimTrails:Particle System Soft Object Reference
    ReturnValue_AnimTrailsNiagara:Niagara System Soft Object Reference
    ReturnValue_NiagaraSkeletonDIVariableName:Name
    ReturnValue_AnimTrailsFirstSocketName:Name
    ReturnValue_AnimTrailsSecondSocketName:Name
    ReturnValue_AnimTrailsWidth:float=~
    ReturnValue_bAttachAnimTrailsToWeapon:bool=~
   )
   n12 = Branch(execute=@n3.then_3, Condition=IsValidSoftObjectReference(SoftObjectReference=@n11.ReturnValue_AnimTrailsNiagara))
   > Set AnimTrail VFX(AnimTrail VFX=@n11.ReturnValue_AnimTrailsNiagara)
   > n13 = Set Use Cascade AnimTrail(Use Cascade AnimTrail:bool=false) -> (Output_Get:bool)
   Set Anim Trails Niagara Skeleton DI Parameter Name(execute=[@n15.then, @n16.then, @n13.then], Anim Trails Niagara Skeleton DI Parameter Name=@n11.ReturnValue_NiagaraSkeletonDIVariableName) -> (Output_Get=None)
   > Set Anim Trails First Socket Name(Anim Trails First Socket Name=@n11.ReturnValue_AnimTrailsFirstSocketName) -> (Output_Get=None)
   > Set Anim Trails Second Socket Name(Anim Trails Second Socket Name=@n11.ReturnValue_AnimTrailsSecondSocketName) -> (Output_Get=None)
   > Set Anim Trails Width(Anim Trails Width=@n11.ReturnValue_AnimTrailsWidth)
   > Set Attach Anim Trails to Weapon(Attach Anim Trails to Weapon=@n11.ReturnValue_bAttachAnimTrailsToWeapon)
   n14 = Branch(execute=@n12.else, Condition=IsValidSoftObjectReference(SoftObjectReference=@n11))
   > Set AnimTrail VFX(AnimTrail VFX=@n11)
   > n15 = Set Use Cascade AnimTrail(Use Cascade AnimTrail=true)
   Set AnimTrail VFX(execute=@n14.else)
   > n16 = Set Use Cascade AnimTrail(Use Cascade AnimTrail=false)
   n17 = GetImpactEffectData!(self:Fort Weapon Melee Item Definition Object Reference=Get Weapon Item Definition()) -> (
    ReturnValue_ImpactPhysicalSurfaceEffectsMap:Map of EPhysicalSurface Enums to Particle System Soft Object References
    ReturnValue_ImpactNiagaraPhysicalSurfaceEffectsMap:Map of EPhysicalSurface Enums to Niagara System Soft Object References
   )
   Set Impact Cascade Physical Surface Effects Map(execute=@n3.then_4, Impact Cascade Physical Surface Effects Map:Map of EPhysicalSurface Enums to Particle System Soft Object References=@n17) -> (Output_Get:Map of EPhysicalSurface Enums to Particle System Soft Object References=SurfaceType_Default)
   > Set Impact Niagara Physical Surface Effects Map(Impact Niagara Physical Surface Effects Map:Map of EPhysicalSurface Enums to Niagara System Soft Object References=@n17.ReturnValue_ImpactNiagaraPhysicalSurfaceEffectsMap) -> (Output_Get:Map of EPhysicalSurface Enums to Niagara System Soft Object References=SurfaceType_Default)
   n18 = Find(
    TargetMap:Map of EPhysicalSurface Enums to Niagara System Soft Object References=@n17.ReturnValue_ImpactNiagaraPhysicalSurfaceEffectsMap
    Key:"EPhysicalSurface Enum (by ref)"=SurfaceType_Default
   ) -> (Value:Niagara System Soft Object Reference)
   > n19 = Branch(Condition=@n18.ReturnValue)
   > Set Impact VFX_Default(Impact VFX_Default=@n18)
   > n20 = Set Use Cascade Impacts(Use Cascade Impacts:bool=false) -> (Output_Get:bool)
   n21 = Branch(
    execute=[@n20.then, @n25.then, @n24.then]
    Condition=IsValidSoftObjectReference(SoftObjectReference=Get Impact VFX_Default() -> (Impact VFX_Default:FXSystem Asset Soft Object Reference))
   )
   > Set Missing Impacts(Missing Impacts:bool=false) -> (Output_Get:bool)
   Set Missing Impacts(execute=@n21.else, Missing Impacts=true)
   n22 = Find(TargetMap:Map of EPhysicalSurface Enums to Particle System Soft Object References=@n17, Key=SurfaceType_Default) -> (Value:Particle System Soft Object Reference)
   n23 = Branch(execute=@n19.else, Condition=@n22.ReturnValue)
   > Set Impact VFX_Default(Impact VFX_Default=@n22)
   > n24 = Set Use Cascade Impacts(Use Cascade Impacts=true)
   Set Impact VFX_Default(execute=@n23.else)
   > n25 = Set Use Cascade Impacts(Use Cascade Impacts=false)
   n26 = GetIdleEffectData!(self:Fort Weapon Melee Item Definition Object Reference=Get Weapon Item Definition()) -> (
    ReturnValue_IdleEffect:Particle System Soft Object Reference
    ReturnValue_IdleEffectNiagara:Niagara System Soft Object Reference
    ReturnValue_IdleFXSocketName:Name
   )
   n27 = Branch(execute=@n3.then_1, Condition=IsValidSoftObjectReference(SoftObjectReference=@n26.ReturnValue_IdleEffectNiagara))
   > Set Idle VFX(Idle VFX=@n26.ReturnValue_IdleEffectNiagara)
   > n28 = Set Use Cascade Idle(Use Cascade Idle:bool=false) -> (Output_Get:bool)
   Set Idle VFX Socket Name(execute=[@n30.then, @n31.then, @n28.then], Idle VFX Socket Name=@n26.ReturnValue_IdleFXSocketName) -> (Output_Get=None)
   n29 = Branch(execute=@n27.else, Condition=IsValidSoftObjectReference(SoftObjectReference=@n26))
   > Set Idle VFX(Idle VFX=@n26)
   > n30 = Set Use Cascade Idle(Use Cascade Idle=true)
   Set Idle VFX(execute=@n29.else)
   > n31 = Set Use Cascade Idle(Use Cascade Idle=false)
  ]}
]
variables [
 Anim Trails First Socket Name {type=string, default=None}
 Anim Trails Niagara Skeleton DI Parameter Name {type=string, default=None}
 Anim Trails Second Socket Name {type=string, default=None}
 Anim Trails Width {type=number}
 AnimTrail VFX {type=FXSystemAsset, default=null}
 Attach Anim Trails to Weapon {type=boolean}
 Idle VFX {type=FXSystemAsset, default=null}
 Idle VFX Socket Name {type=string, default=None}
 Impact Cascade Physical Surface Effects Map {default {}}
 Impact Niagara Physical Surface Effects Map {default {}}
 Impact VFX_Default {type=FXSystemAsset, default=null}
 Log Changes {default=()}
 Missing Impacts {type=boolean}
 Pickaxe Changed {default=()}
 Pickaxe ID {type=AthenaPickaxeItemDefinition, default=null}
 Pickaxe ID_Setter {type=AthenaPickaxeItemDefinition, default=null}
 Pickaxe Weapon Actor Class {type=Class, default=null}
 ShowAdvancedImpactsSetup {type=boolean}
 Swing VFX {type=FXSystemAsset, default=null}
 Swing VFX Socket Name {type=string, default=None}
 Use Cascade AnimTrail {type=boolean}
 Use Cascade Idle {type=boolean}
 Use Cascade Impacts {type=boolean}
 Use Cascade Swing {type=boolean}
 Weapon Item Definition {type=FortWeaponMeleeItemDefinition, default=null}
 Weapon Item Definition_Setter {type=FortWeaponMeleeItemDefinition, default=null}
]
widgets {
 info {rootWidgetClass=&StackBox, widgetCount=5}
 widgets [
  StackBox_0 {widgetClassPath=&StackBox, widget=StackBox_0}
  CommonHierarchicalScrollBox_98 {
   widgetClassPath=/Script/CommonUI.CommonHierarchicalScrollBox
   parent=StackBox_0
   widget=CommonHierarchicalScrollBox_98
   properties {
    widgetBarStyle {
     thickness=8
     draggedThumbImage {
      bIsDynamicallyLoaded=false
      drawAs=RoundedBox
      imageType=NoImage
      mirroring=NoMirror
      resourceName=None
      resourceObject=null
      tiling=NoTile
      imageSize=<0 0>
      margin=<ltrb 0 0 0 0>
      outlineSettings {
       bUseBrushTransparency=false
       roundingType=FixedRadius
       width=0
       color {colorUseRule=UseColor_Specified, specifiedColor=<rgba 0 0 0 0>}
       cornerRadii=<4 4 4 4>
      }
      tintColor {colorUseRule=UseColor_Specified, specifiedColor=<rgba 0.2159 0.2159 0.2159 1>}
      uVRegion {bIsValid=false, max=<0 0>, min=<0 0>}
     }
     horizontalBackgroundImage {
      bIsDynamicallyLoaded=false
      drawAs=NoDrawType
      imageType=NoImage
      mirroring=NoMirror
      resourceName=None
      resourceObject=null
      tiling=NoTile
      imageSize=<0 0>
      margin=<ltrb 0 0 0 0>
      outlineSettings {
       bUseBrushTransparency=false
       roundingType=HalfHeightRadius
       width=0
       color {colorUseRule=UseColor_Specified, specifiedColor=<rgba 0 0 0 0>}
       cornerRadii=<0 0 0 0>
      }
      tintColor {colorUseRule=UseColor_Specified, specifiedColor=<rgba 1 1 1 1>}
      uVRegion {bIsValid=false, max=<0 0>, min=<0 0>}
     }
     horizontalBottomSlotImage {
      bIsDynamicallyLoaded=false
      drawAs=NoDrawType
      imageType=NoImage
      mirroring=NoMirror
      resourceName=None
      resourceObject=null
      tiling=NoTile
      imageSize=<0 0>
      margin=<ltrb 0 0 0 0>
      outlineSettings {
       bUseBrushTransparency=false
       roundingType=HalfHeightRadius
       width=0
       color {colorUseRule=UseColor_Specified, specifiedColor=<rgba 0 0 0 0>}
       cornerRadii=<0 0 0 0>
      }
      tintColor {colorUseRule=UseColor_Specified, specifiedColor=<rgba 1 1 1 1>}
      uVRegion {bIsValid=false, max=<0 0>, min=<0 0>}
     }
     horizontalTopSlotImage {
      bIsDynamicallyLoaded=false
      drawAs=NoDrawType
      imageType=NoImage
      mirroring=NoMirror
      resourceName=None
      resourceObject=null
      tiling=NoTile
      imageSize=<0 0>
      margin=<ltrb 0 0 0 0>
      outlineSettings {
       bUseBrushTransparency=false
       roundingType=HalfHeightRadius
       width=0
       color {colorUseRule=UseColor_Specified, specifiedColor=<rgba 0 0 0 0>}
       cornerRadii=<0 0 0 0>
      }
      tintColor {colorUseRule=UseColor_Specified, specifiedColor=<rgba 1 1 1 1>}
      uVRegion {bIsValid=false, max=<0 0>, min=<0 0>}
     }
     hoveredThumbImage {
      bIsDynamicallyLoaded=false
      drawAs=RoundedBox
      imageType=NoImage
      mirroring=NoMirror
      resourceName=None
      resourceObject=null
      tiling=NoTile
      imageSize=<0 0>
      margin=<ltrb 0 0 0 0>
      outlineSettings {
       bUseBrushTransparency=false
       roundingType=FixedRadius
       width=0
       color {colorUseRule=UseColor_Specified, specifiedColor=<rgba 0 0 0 0>}
       cornerRadii=<4 4 4 4>
      }
      tintColor {colorUseRule=UseColor_Specified, specifiedColor=<rgba 0.2159 0.2159 0.2159 1>}
      uVRegion {bIsValid=false, max=<0 0>, min=<0 0>}
     }
     normalThumbImage {
      bIsDynamicallyLoaded=false
      drawAs=RoundedBox
      imageType=NoImage
      mirroring=NoMirror
      resourceName=None
      resourceObject=null
      tiling=NoTile
      imageSize=<0 0>
      margin=<ltrb 0 0 0 0>
      outlineSettings {
       bUseBrushTransparency=false
       roundingType=FixedRadius
       width=0
       color {colorUseRule=UseColor_Specified, specifiedColor=<rgba 0 0 0 0>}
       cornerRadii=<4 4 4 4>
      }
      tintColor {colorUseRule=UseColor_Specified, specifiedColor=<rgba 0.0953 0.0953 0.0953 1>}
      uVRegion {bIsValid=false, max=<0 0>, min=<0 0>}
     }
     verticalBackgroundImage {
      bIsDynamicallyLoaded=false
      drawAs=NoDrawType
      imageType=NoImage
      mirroring=NoMirror
      resourceName=None
      resourceObject=null
      tiling=NoTile
      imageSize=<0 0>
      margin=<ltrb 0 0 0 0>
      outlineSettings {
       bUseBrushTransparency=false
       roundingType=HalfHeightRadius
       width=0
       color {colorUseRule=UseColor_Specified, specifiedColor=<rgba 0 0 0 0>}
       cornerRadii=<0 0 0 0>
      }
      tintColor {colorUseRule=UseColor_Specified, specifiedColor=<rgba 1 1 1 1>}
      uVRegion {bIsValid=false, max=<0 0>, min=<0 0>}
     }
     verticalBottomSlotImage {
      bIsDynamicallyLoaded=false
      drawAs=NoDrawType
      imageType=NoImage
      mirroring=NoMirror
      resourceName=None
      resourceObject=null
      tiling=NoTile
      imageSize=<0 0>
      margin=<ltrb 0 0 0 0>
      outlineSettings {
       bUseBrushTransparency=false
       roundingType=HalfHeightRadius
       width=0
       color {colorUseRule=UseColor_Specified, specifiedColor=<rgba 0 0 0 0>}
       cornerRadii=<0 0 0 0>
      }
      tintColor {colorUseRule=UseColor_Specified, specifiedColor=<rgba 1 1 1 1>}
      uVRegion {bIsValid=false, max=<0 0>, min=<0 0>}
     }
     verticalTopSlotImage {
      bIsDynamicallyLoaded=false
      drawAs=NoDrawType
      imageType=NoImage
      mirroring=NoMirror
      resourceName=None
      resourceObject=null
      tiling=NoTile
      imageSize=<0 0>
      margin=<ltrb 0 0 0 0>
      outlineSettings {
       bUseBrushTransparency=false
       roundingType=HalfHeightRadius
       width=0
       color {colorUseRule=UseColor_Specified, specifiedColor=<rgba 0 0 0 0>}
       cornerRadii=<0 0 0 0>
      }
      tintColor {colorUseRule=UseColor_Specified, specifiedColor=<rgba 1 1 1 1>}
      uVRegion {bIsValid=false, max=<0 0>, min=<0 0>}
     }
    }
    widgetStyle {
     barThickness=8
     bottomShadowBrush {
      bIsDynamicallyLoaded=false
      drawAs=Box
      imageType=FullColor
      mirroring=NoMirror
      resourceName=../../../Engine/Content/Slate/Common/ScrollBoxShadowBottom.png
      resourceObject=null
      tiling=NoTile
      imageSize=<16 8>
      margin=<ltrb 0.5 0 0.5 1>
      outlineSettings {
       bUseBrushTransparency=false
       roundingType=HalfHeightRadius
       width=0
       color {colorUseRule=UseColor_Specified, specifiedColor=<rgba 0 0 0 0>}
       cornerRadii=<0 0 0 0>
      }
      tintColor {colorUseRule=UseColor_Specified, specifiedColor=<rgba 1 1 1 1>}
      uVRegion {bIsValid=false, max=<0 0>, min=<0 0>}
     }
     horizontalScrolledContentPadding=<ltrb 0 0 1 0>
     leftShadowBrush {
      bIsDynamicallyLoaded=false
      drawAs=Box
      imageType=FullColor
      mirroring=NoMirror
      resourceName=../../../Engine/Content/Slate/Common/ScrollBoxShadowLeft.png
      resourceObject=null
      tiling=NoTile
      imageSize=<8 16>
      margin=<ltrb 1 0.5 0 0.5>
      outlineSettings {
       bUseBrushTransparency=false
       roundingType=HalfHeightRadius
       width=0
       color {colorUseRule=UseColor_Specified, specifiedColor=<rgba 0 0 0 0>}
       cornerRadii=<0 0 0 0>
      }
      tintColor {colorUseRule=UseColor_Specified, specifiedColor=<rgba 1 1 1 1>}
      uVRegion {bIsValid=false, max=<0 0>, min=<0 0>}
     }
     rightShadowBrush {
      bIsDynamicallyLoaded=false
      drawAs=Box
      imageType=FullColor
      mirroring=NoMirror
      resourceName=../../../Engine/Content/Slate/Common/ScrollBoxShadowRight.png
      resourceObject=null
      tiling=NoTile
      imageSize=<8 16>
      margin=<ltrb 0 0.5 1 0.5>
      outlineSettings {
       bUseBrushTransparency=false
       roundingType=HalfHeightRadius
       width=0
       color {colorUseRule=UseColor_Specified, specifiedColor=<rgba 0 0 0 0>}
       cornerRadii=<0 0 0 0>
      }
      tintColor {colorUseRule=UseColor_Specified, specifiedColor=<rgba 1 1 1 1>}
      uVRegion {bIsValid=false, max=<0 0>, min=<0 0>}
     }
     topShadowBrush {
      bIsDynamicallyLoaded=false
      drawAs=Box
      imageType=FullColor
      mirroring=NoMirror
      resourceName=../../../Engine/Content/Slate/Common/ScrollBoxShadowTop.png
      resourceObject=null
      tiling=NoTile
      imageSize=<16 8>
      margin=<ltrb 0.5 1 0.5 0>
      outlineSettings {
       bUseBrushTransparency=false
       roundingType=HalfHeightRadius
       width=0
       color {colorUseRule=UseColor_Specified, specifiedColor=<rgba 0 0 0 0>}
       cornerRadii=<0 0 0 0>
      }
      tintColor {colorUseRule=UseColor_Specified, specifiedColor=<rgba 1 1 1 1>}
      uVRegion {bIsValid=false, max=<0 0>, min=<0 0>}
     }
     verticalScrolledContentPadding=<ltrb 0 0 0 1>
    }
   }
   slot {class=StackBoxSlot, properties {size {sizeRule=Fill, value=1}}}
  }
  DetailsView_WID {
   widgetClassPath=&DetailsView
   parent=CommonHierarchicalScrollBox_98
   widget=DetailsView_WID
   properties {bAllowFiltering=true, bAutoLoadAsset=false, bShowModifiedPropertiesOption=true, categoriesToShow [Setup]}
   slot {class=ScrollBoxSlot, properties {verticalAlignment=VAlign_Top}}
  }
  Spacer {
   widgetClassPath=&Spacer
   parent=CommonHierarchicalScrollBox_98
   widget=Spacer
   properties {size=<5 5>}
   slot {class=ScrollBoxSlot}
  }
  DetailsView_WID_Advanced {
   widgetClassPath=&DetailsView
   parent=CommonHierarchicalScrollBox_98
   widget=DetailsView_WID_Advanced
   properties {bAllowFiltering=true, bAutoLoadAsset=false, bShowModifiedPropertiesOption=true, categoriesToShow [Advanced_Setup]}
   slot {class=ScrollBoxSlot, properties {verticalAlignment=VAlign_Top}}
  }
 ]
}
