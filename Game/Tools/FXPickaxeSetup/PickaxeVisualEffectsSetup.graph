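# Editor utility widget that writes pickaxe VFX (actor data, idle, swing, anim trail, impact) into weapon item definitions (WIDs).
# The selected pickaxe's WID type shows either the single or dual-wield setup sub-widget.
# The button first validates default impact VFX, then checks out and saves the WID, also saving the pickaxe definition on WID mismatch.
# Effects are routed by Niagara or Cascade cast, with results logged by severity.
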EditorUtilityWidgetBlueprint /Game/Tools/FXPickaxeSetup/PickaxeVisualEffectsSetup
graphs [
 Checkout and Save Pickaxe ID {nodes [
   n1 = Checkout and Save Pickaxe ID!() -> (then, Dual Wield ?:bool=~) {is=title K2Node_FunctionEntry}
   > n2 = Branch(Condition=@n1)
   > n3 = CheckoutLoadedAsset!(
    execute
    AssetToCheckout:Object Reference=Get Pickaxe ID!(self:EUW WID FX Dual Wield WID Setup Object Reference=Get EUW_WID_FX_DualWieldWID_Setup() -> (EUW_WID_FX_DualWieldWID_Setup:EUW WID FX Dual Wield WID Setup Object Reference)) -> (Pickaxe ID:Athena Pickaxe Item Definition Object Reference)
   ) -> (then, ReturnValue:bool)
   > n4 = Branch(Condition=@n3)
   > Logging Text!(execute, self, Message for Log:Text="LOCTEXT(\"Checking out Pickaxe ID File, due to mismatch WID file\")", Error Code:int=0) -> (then)
   > n5 = SaveLoadedAsset(AssetToSave=Get Pickaxe ID(self=Get EUW_WID_FX_DualWieldWID_Setup()), bOnlyIfIsDirty=false)
   > n6 = Branch(Condition=@n5)
   > Logging Text(Message for Log="LOCTEXT(\"Saving Pickaxe ID File with new WID file\")", Error Code=0)
   > Return Node!(execute, Success ?:bool=true)
   Logging Text(
    execute=@n6.else
    Message for Log="LOCTEXT(\"Error - Attempted to save Pickaxe ID File, due to mismatch WID file, but failed.  Please Check your Source Control.\")"
    Error Code=2
   )
   > Large Warning Setup!(execute, self, Error Code:int=2) -> (then)
   > Return Node(Success ?=false)
   Logging Text(
    execute=@n4.else
    Message for Log="LOCTEXT(\"Error - Attempted to Check out Pickaxe ID File, due to mismatch WID file, but failed.  Please Check your Source Control.\")"
    Error Code=2
   )
   > Large Warning Setup(Error Code=2)
   > Return Node(Success ?=false)
   n7 = CheckoutLoadedAsset(
    execute=@n2.else
    AssetToCheckout=Get Pickaxe ID(self:EUW WID FX WID Setup Object Reference=Get EUW_WID_FX_WID_Setup() -> (EUW_WID_FX_WID_Setup:EUW WID FX WID Setup Object Reference))
   )
   > n8 = Branch(Condition=@n7)
   > Logging Text(Message for Log="LOCTEXT(\"Checking out Pickaxe ID File, due to mismatch WID file\")", Error Code=0)
   > n9 = SaveLoadedAsset(AssetToSave=Get Pickaxe ID(self=Get EUW_WID_FX_WID_Setup()), bOnlyIfIsDirty=false)
   > n10 = Branch(Condition=@n9)
   > Logging Text(Message for Log="LOCTEXT(\"Saving Pickaxe ID File with new WID file\")", Error Code=0)
   > Return Node(Success ?=true)
   Logging Text(
    execute=@n10.else
    Message for Log="LOCTEXT(\"Error - Attempted to save Pickaxe ID File, due to mismatch WID file, but failed.  Please Check your Source Control.\")"
    Error Code=2
   )
   > Large Warning Setup(Error Code=2)
   > Return Node(Success ?=false)
   Logging Text(
    execute=@n8.else
    Message for Log="LOCTEXT(\"Error - Attempted to Check out Pickaxe ID File, due to mismatch WID file, but failed.  Please Check your Source Control.\")"
    Error Code=2
   )
   > Large Warning Setup(Error Code=2)
   > Return Node(Success ?=false)
  ]}
 Clear Logging Text {nodes [
   Clear Logging Text() {is=title K2Node_FunctionEntry}
   > Set Text(self=Get Messages() -> (Messages:Text Object Reference))
   > Set Text(Text="LOCTEXT(\"Warnings\")", self=Get Warnings() -> (Warnings:Text Object Reference))
   > Set Text(Text="LOCTEXT(\"Errors\")", self=Get Errors() -> (Errors:Text Object Reference))
   > SetVisibility(self=Get Messages(), InVisibility=Collapsed)
   > SetVisibility(self=Get Warnings(), InVisibility=Collapsed)
   > SetVisibility(self=Get Errors(), InVisibility=Collapsed)
   > SetVisibility(
    self=[Get WarningIcon() -> (WarningIcon:Image Object Reference), Get LargeWarningText() -> (LargeWarningText:Text Object Reference)]
    InVisibility=Collapsed
   )
   n1 = Get Validation() -> (Validation:Text Object Reference)
   > Set Text(Text="LOCTEXT(\"Validate Changes\")", self=@n1)
   > SetColorAndOpacity(self=@n1, InColorAndOpacity=Get Base Color() -> (Base Color:Slate Color Structure))
  ]}
 EventGraph {nodes [
   n1 = "On Property Changed (Initial_PickaxeSelection)"!() -> (OutputDelegate, then, PropertyName:Name=None) {is=title K2Node_ComponentBoundEvent}
   n2 = Get WeaponDefinition!(self:Athena Pickaxe Item Definition Object Reference=Get Pickaxe_ID() -> (Pickaxe_ID:Athena Pickaxe Item Definition Object Reference)) -> (WeaponDefinition:Fort Weapon Melee Item Definition Object Reference)
   n3 = Is Valid(exec=[@n49.then, @n1.then], InputObject=@n2)
   Logging Text(
    execute=@n3.Is Not Valid
    Message for Log="LOCTEXT(\"Error - Missing Weapon Item Definition in Pickaxe_ID.  Please contact Implementation\")"
    Error Code=2
   )
   > Large Warning Setup(Error Code=2)
   > WID Setup!(execute, self, Dual Wield ?:bool=false) -> (then)
   > WID Setup(Dual Wield ?=true)
   n4 = "IsA ( soft )"!(
    execute=@n3.Is Valid
    Object:Object Reference=@n2
    SoftClass:Object Soft Class Reference=/Script/FortniteGame.FortWeaponMeleeDualWieldItemDefinition
   ) -> (True:Exec, False:Exec)
   > WID Setup(Dual Wield ?=true)
   > SetVisibility(self=Get EUW_WID_FX_DualWieldWID_Setup(), InVisibility=Visible)
   > SetVisibility(self=Get EUW_WID_FX_WID_Setup(), InVisibility=Collapsed)
   > SetVisibility(self=Get Initial_PickaxeSelection() -> (Initial_PickaxeSelection:Details View Object Reference), InVisibility=Collapsed)
   > Set isDualWield?(isDualWield?:bool=true) -> (Output_Get:bool)
   n5 = Create Event()
   n6 = Get EUW_WID_FX_WID_Setup()
   > Unbind Event from Pickaxe Changed(self:EUW WID FX WID Setup Object Reference=@n6, Delegate=@n5) {is=title K2Node_RemoveDelegate}
   > Unbind Event from Log Changes(self:EUW WID FX WID Setup Object Reference=@n6, Delegate=Create Event()) {is=title K2Node_RemoveDelegate}
   n7 = Get EUW_WID_FX_DualWieldWID_Setup()
   > Bind Event to Pickaxe Changed(self:EUW WID FX Dual Wield WID Setup Object Reference=@n7, Delegate=@n5) {is=title K2Node_AddDelegate}
   > Bind Event to Log Changes(self:EUW WID FX Dual Wield WID Setup Object Reference=@n7, Delegate=Create Event()) {is=title K2Node_AddDelegate}
   WID Setup(execute=@n4.False, Dual Wield ?=false)
   > SetVisibility(self=Get EUW_WID_FX_DualWieldWID_Setup(), InVisibility=Collapsed)
   > SetVisibility(self=Get EUW_WID_FX_WID_Setup(), InVisibility=Visible)
   > SetVisibility(self=Get Initial_PickaxeSelection(), InVisibility=Collapsed)
   > Set isDualWield?(isDualWield?=false)
   n8 = Create Event()
   n9 = Get EUW_WID_FX_WID_Setup()
   > Bind Event to Pickaxe Changed(self:EUW WID FX WID Setup Object Reference=@n9, Delegate=@n8)
   > Bind Event to Log Changes(self:EUW WID FX WID Setup Object Reference=@n9, Delegate=Create Event())
   n10 = Get EUW_WID_FX_DualWieldWID_Setup()
   > Unbind Event from Pickaxe Changed(self:EUW WID FX Dual Wield WID Setup Object Reference=@n10, Delegate=@n8)
   > Unbind Event from Log Changes(self:EUW WID FX Dual Wield WID Setup Object Reference=@n10, Delegate=Create Event())
   "On Clicked (Interact_Button)"() {is=title K2Node_ComponentBoundEvent}
   > n11 = Branch(Condition=Get Allow Interaction() -> (Allow Interaction:bool))
   > Clear Logging Text() {is=title K2Node_CallFunction}
   > n12 = Branch(Condition=Get isDualWield?() -> (isDualWield?:bool))
   n13 = Get EUW_WID_FX_DualWieldWID_Setup()
   n14 = Get Pickaxe ID(self:EUW WID FX Dual Wield WID Setup Object Reference=@n13)
   > n15 = Branch(Condition="Equal (Object)"(
     A=Get Dual Wield Weapon Item Definition!(self:EUW WID FX Dual Wield WID Setup Object Reference=@n13) -> (Dual Wield Weapon Item Definition:Fort Weapon Melee Dual Wield Item Definition Object Reference)
     B=Get WeaponDefinition(self=@n14)
    ))
   n16 = CheckoutLoadedAsset(execute=[@n15.then, @n26.then], AssetToCheckout=Get Dual Wield Weapon Item Definition(self=Get EUW_WID_FX_DualWieldWID_Setup()))
   > n17 = Branch(Condition=@n16)
   > Logging Text(Message for Log="LOCTEXT(\"Checking out WID File and beginning Item Data setup\")", Error Code=0)
   > n18 = Set Pickaxe Actor Data!(execute, self, Dual Wield ?:bool=true) -> (then, Success ?:bool) {is=title K2Node_CallFunction}
   > Branch(Condition=@n18)
   > n19 = Set Idle Item Data!(execute, self, Dual Wield ?:bool=true) -> (then, Success ?:bool) {is=title K2Node_CallFunction}
   > Branch(Condition=@n19)
   > n20 = Set Swing Item Data!(execute, self, Dual Wield ?:bool=true) -> (then, Success ?:bool) {is=title K2Node_CallFunction}
   > Branch(Condition=@n20)
   > n21 = Set AnimTrail Item Data!(execute, self, Dual Wield ?:bool=true) -> (then, Success ?:bool) {is=title K2Node_CallFunction}
   > Branch(Condition=@n21)
   > n22 = Set Impact Item Data!(execute, self, Dual Wield ?:bool=true) -> (then, Success ?:bool) {is=title K2Node_CallFunction}
   > Branch(Condition=@n22)
   > n23 = SaveLoadedAsset(AssetToSave=Get Dual Wield Weapon Item Definition(self=Get EUW_WID_FX_DualWieldWID_Setup()), bOnlyIfIsDirty=false)
   > n24 = Branch(Condition=@n23)
   > Logging Text(Message for Log="LOCTEXT(\"WID Saved with updated setup\")", Error Code=0)
   Logging Text(
    execute=@n24.else
    Message for Log="LOCTEXT(\"WID was NOT saved with updated setup, please check source control or other errors.\")"
    Error Code=2
   )
   > Large Warning Setup(Error Code=2)
   Logging Text(
    execute=@n17.else
    Message for Log="LOCTEXT(\"ERROR - Attempted to Check out WID File, but failed.  Please Check your Source Control.\")"
    Error Code=2
   )
   > Large Warning Setup(Error Code=2)
   n25 = Checkout and Save Pickaxe ID!(execute=@n15.else, self, Dual Wield ?:bool=true) -> (then, Success ?:bool) {is=title K2Node_CallFunction}
   > n26 = Branch(Condition=@n25)
   n27 = Get EUW_WID_FX_WID_Setup()
   n28 = Get Pickaxe ID(self:EUW WID FX WID Setup Object Reference=@n27)
   n29 = Branch(
    execute=@n12.else
    Condition="Equal (Object)"(
     A=Get Weapon Item Definition!(self:EUW WID FX WID Setup Object Reference=@n27) -> (Weapon Item Definition:Fort Weapon Melee Item Definition Object Reference)
     B=Get WeaponDefinition(self=@n28)
    )
   )
   n30 = CheckoutLoadedAsset(execute=[@n29.then, @n40.then], AssetToCheckout=Get Weapon Item Definition(self=Get EUW_WID_FX_WID_Setup()))
   > n31 = Branch(Condition=@n30)
   > Logging Text(Message for Log="LOCTEXT(\"Checking out WID File and beginning Item Data setup\")", Error Code=0)
   > n32 = Set Pickaxe Actor Data(Dual Wield ?=false)
   > Branch(Condition=@n32)
   > n33 = Set Idle Item Data(Dual Wield ?=false)
   > Branch(Condition=@n33)
   > n34 = Set Swing Item Data(Dual Wield ?=false)
   > Branch(Condition=@n34)
   > n35 = Set AnimTrail Item Data(Dual Wield ?=false)
   > Branch(Condition=@n35)
   > n36 = Set Impact Item Data(Dual Wield ?=false)
   > Branch(Condition=@n36)
   > n37 = SaveLoadedAsset(AssetToSave=Get Weapon Item Definition(self=Get EUW_WID_FX_WID_Setup()), bOnlyIfIsDirty=false)
   > n38 = Branch(Condition=@n37)
   > Logging Text(Message for Log="LOCTEXT(\"WID Saved with updated setup\")", Error Code=0)
   Logging Text(
    execute=@n38.else
    Message for Log="LOCTEXT(\"WID was NOT saved with updated setup, please check source control or other errors.\")"
    Error Code=2
   )
   > Large Warning Setup(Error Code=2)
   Logging Text(
    execute=@n31.else
    Message for Log="LOCTEXT(\"ERROR - Attempted to Check out WID File, but failed.  Please Check your Source Control.\")"
    Error Code=2
   )
   > Large Warning Setup(Error Code=2)
   n39 = Checkout and Save Pickaxe ID(execute=@n29.else, Dual Wield ?=false)
   > n40 = Branch(Condition=@n39)
   Clear Logging Text(execute=@n11.else)
   > n41 = Branch(Condition=Get isDualWield?())
   > n42 = LoadAsset_Blocking(Asset=Get Impact VFX_Default!(self:EUW WID FX Dual Wield WID Setup Object Reference=Get EUW_WID_FX_DualWieldWID_Setup()) -> (Impact VFX_Default:FXSystem Asset Soft Object Reference))
   > n43 = Is Valid(InputObject=@n42)
   Logging Text(
    execute=@n43.Is Not Valid
    Message for Log="LOCTEXT(\"Error - Impact VFX - Default is not Valid, Validation Check has FAILED.\")"
    Error Code=2
   )
   > Large Warning Setup(Error Code=2)
   n44 = LoadAsset_Blocking(
    execute=@n43.Is Valid
    Asset=Get Offhand Impact VFX_Default!(self:EUW WID FX Dual Wield WID Setup Object Reference=Get EUW_WID_FX_DualWieldWID_Setup()) -> (Offhand Impact VFX_Default:FXSystem Asset Soft Object Reference)
   )
   > n45 = Is Valid(InputObject=@n44)
   Logging Text(
    execute=@n45.Is Not Valid
    Message for Log="LOCTEXT(\"Error - Offhand Impact VFX - Default is not Valid, Validation Check has FAILED.\")"
    Error Code=2
   )
   > Large Warning Setup(Error Code=2)
   Large Warning Setup(execute=@n45.Is Valid, Error Code=0)
   > Logging Text(Message for Log="LOCTEXT(\"Validation Successful - Press again to Save.\")", Error Code=0)
   n46 = LoadAsset_Blocking(execute=@n41.else, Asset=Get Impact VFX_Default(self:EUW WID FX WID Setup Object Reference=Get EUW_WID_FX_WID_Setup()))
   > n47 = Is Valid(InputObject=@n46)
   Logging Text(
    execute=@n47.Is Not Valid
    Message for Log="LOCTEXT(\"Error - Impact VFX - Default is not Valid, Validation Check has FAILED.\")"
    Error Code=2
   )
   > Large Warning Setup(Error Code=2)
   Large Warning Setup(execute=@n47.Is Valid, Error Code=0)
   > Logging Text(Message for Log="LOCTEXT(\"Validation Successful - Press again to Save.\")", Error Code=0)
   n48 = Set Pickaxe ID!() -> (OutputDelegate, then, PIckaxeID:Athena Pickaxe Item Definition Object Reference) {is=title K2Node_CustomEvent}
   > Set Pickaxe_ID(Pickaxe_ID:Athena Pickaxe Item Definition Object Reference=@n48.PIckaxeID) -> (Output_Get:Athena Pickaxe Item Definition Object Reference)
   > Called Setup()
   Called Setup() {is=title K2Node_CustomEvent}
   > n49 = Clear Logging Text()
   Logging_DualWield() {is=title K2Node_CustomEvent}
   > WIDValidationLogging!(execute, self, Dual Wield ?:bool=true) -> (then)
   Logging() {is=title K2Node_CustomEvent}
   > WIDValidationLogging(Dual Wield ?=false)
   Event PreConstruct()
   > SetObject(self=Get Initial_PickaxeSelection(), NewObject=Self-Reference())
   > SetVisibility(self=[Get EUW_WID_FX_WID_Setup(), Get EUW_WID_FX_DualWieldWID_Setup()], InVisibility=Collapsed)
  ]}
 Large Warning Setup {nodes [
   n1 = Large Warning Setup!() -> (then, Error Code:int) {is=title K2Node_FunctionEntry}
   > SetVisibility(self=[Get WarningIcon(), Get LargeWarningText()], InVisibility=Visible)
   > n2 = Switch on Int!(execute:Exec, Selection:int=@n1) -> (Default:Exec, 0:Exec, 1:Exec, 2:Exec)
   n3 = Get WarningIcon()
   SetBrushFromTexture!(
    execute=@n2.0
    self:Image Object Reference=@n3
    Texture:Texture 2D Object Reference=/Game/UI/Foundation/Textures/Icons/Checks/T-Icon-Check-128
    bMatchSize:bool=true
   ) -> (then)
   > SetBrushTintColor!(execute, self:Image Object Reference=@n3, TintColor:Slate Color Structure=Get Good Color() -> (Good Color:Slate Color Structure)) -> (then)
   > SetDesiredSizeOverride!(execute, self:Image Object Reference=@n3, DesiredSize:Vector 2D Structure=(X=32,Y=32)) -> (then)
   n4 = Get LargeWarningText()
   > Set Text(Text="LOCTEXT(\"Good to Go\")", self=@n4)
   > Set ColorAndOpacity!(execute, ColorAndOpacity:Slate Color Structure=Get Good Color(), self:Text Object Reference=@n4) -> (then, Output_Get:Slate Color Structure)
   n5 = Get Validation()
   > Set Text(Text="LOCTEXT(\"Save\")", self=@n5)
   > SetColorAndOpacity(self=@n5, InColorAndOpacity=Get Good Color())
   > Set Allow Interaction(Allow Interaction:bool=true) -> (Output_Get:bool)
   n6 = Get WarningIcon()
   SetBrushFromTexture(execute=@n2.1, self=@n6, Texture=/Game/UI/Foundation/Shell/Textures/T-Icon-Warning-128, bMatchSize=true)
   > SetBrushTintColor(self=@n6, TintColor=Get Warning Color() -> (Warning Color:Slate Color Structure))
   > SetDesiredSizeOverride(self=@n6, DesiredSize=(X=32,Y=32))
   n7 = Get LargeWarningText()
   > Set Text(Text="LOCTEXT(\"WARNING\")", self=@n7)
   > Set ColorAndOpacity(ColorAndOpacity=Get Warning Color(), self=@n7)
   n8 = Get Validation()
   > Set Text(Text="LOCTEXT(\"Save\")", self=@n8)
   > SetColorAndOpacity(self=@n8, InColorAndOpacity=Get Warning Color())
   > Set Allow Interaction(Allow Interaction=true)
   SetBrushFromTexture(execute=@n2.2, self=@n6, Texture=&Cancel_Normal, bMatchSize=true)
   > SetBrushTintColor(self=@n6, TintColor=Get Error Color() -> (Error Color:Slate Color Structure))
   > SetDesiredSizeOverride(self=@n6, DesiredSize=(X=32,Y=32))
   > Set Text(Text="LOCTEXT(\"ERROR\")", self=@n7)
   > Set ColorAndOpacity(ColorAndOpacity=Get Error Color(), self=@n7)
   n9 = Get Validation()
   > Set Text(Text="LOCTEXT(\"Validate Changes\")", self=@n9)
   > SetColorAndOpacity(self=@n9, InColorAndOpacity=Get Error Color())
   > Set Allow Interaction(Allow Interaction=false)
  ]}
 Logging Text {nodes [
   n1 = Logging Text!() -> (then, Message for Log:Text, Error Code:int) {is=title K2Node_FunctionEntry}
   n2 = Select!(
    Option 0:Text Object Reference=Get Messages()
    Option 1:Text Object Reference=Get Warnings()
    Option 2:Text Object Reference=Get Errors()
    Index:int=@n1.Error Code
   ) -> (ReturnValue:Text Object Reference)
   n3 = "To String (Text)"(InText=Get Text(self:Text Object Reference=@n2))
   n4 = "To String (Text)"(InText=@n1)
   > Set Text(Text="To Text (String)"(InString=Append!(A=@n3, B="\r\n", C:String=@n4) -> (ReturnValue)), self=@n2)
   > n5 = Switch on Int(Selection=@n1.Error Code)
   SetVisibility(execute=@n5.1, self=Get Warnings(), InVisibility=Visible)
   SetVisibility(execute=@n5.2, self=Get Errors(), InVisibility=Visible)
   SetVisibility(execute=@n5.0, self=Get Messages(), InVisibility=Visible)
  ]}
 Set AnimTrail Item Data {nodes [
   n1 = Set AnimTrail Item Data!() -> (then, Dual Wield ?:bool=~) {is=title K2Node_FunctionEntry}
   > n2 = Branch(Condition=@n1)
   n3 = Get EUW_WID_FX_DualWieldWID_Setup()
   > n4 = Branch(Condition=IsValidSoftObjectReference(SoftObjectReference=Get AnimTrail VFX!(self:EUW WID FX Dual Wield WID Setup Object Reference=@n3) -> (AnimTrail VFX:FXSystem Asset Soft Object Reference)))
   > n5 = LoadAsset_Blocking(Asset=Get AnimTrail VFX(self=Get EUW_WID_FX_DualWieldWID_Setup()))
   > n6 = Cast To NiagaraSystem!(execute, Object=@n5) -> (then, CastFailed, AsNiagara System:Niagara System Object Reference) {is=title K2Node_DynamicCast}
   > SetAnimTrailData!(
    execute
    self:Fort Weapon Melee Item Definition Object Reference=Get Dual Wield Weapon Item Definition(self=Get EUW_WID_FX_DualWieldWID_Setup())
    InAnimTrailData:"Fortnite Weapon Main Hand Melee Anim Trails Item Data Structure (by ref)"=Make Fortnite Weapon Main Hand Melee Anim Trails Item Data!(
     AnimTrails:Particle System Soft Object Reference
     AnimTrailsNiagara:Niagara System Soft Object Reference=Make Soft Reference!(Input:Niagara System Object Reference=@n6) -> (Output:Niagara System Soft Object Reference)
     NiagaraSkeletonDIVariableName:Name=User.SkeletalMesh
     AnimTrailsFirstSocketName:Name
     AnimTrailsSecondSocketName:Name
     AnimTrailsWidth:float=1
     bAttachAnimTrailsToWeapon:bool=true
    ) -> (FortItemComponentData_WeaponMeleeAnimTrails_MainHand:Fortnite Weapon Main Hand Melee Anim Trails Item Data Structure) {is=title K2Node_MakeStruct}
   ) -> (then)
   > n7 = Logging Text(Message for Log="LOCTEXT(\"Niagara Anim Trail VFX was set in WID file.\")", Error Code=0) {is=title K2Node_CallFunction}
   n8 = Get EUW_WID_FX_DualWieldWID_Setup()
   n9 = Branch(
    execute=[@n7.then, @n16.then, @n17.then]
    Condition=IsValidSoftObjectReference(SoftObjectReference=Get Offhand AnimTrail VFX!(self:EUW WID FX Dual Wield WID Setup Object Reference=@n8) -> (Offhand AnimTrail VFX:FXSystem Asset Soft Object Reference))
   )
   > n10 = LoadAsset_Blocking(Asset=Get Offhand AnimTrail VFX(self=Get EUW_WID_FX_DualWieldWID_Setup()))
   > n11 = Cast To NiagaraSystem(Object=@n10)
   > SetAnimTrailDataOffhand!(
    execute
    self:Fort Weapon Melee Dual Wield Item Definition Object Reference=Get Dual Wield Weapon Item Definition(self=Get EUW_WID_FX_DualWieldWID_Setup())
    InAnimTrailData:"Fortnite Weapon Off Hand Melee Anim Trails Item Data Structure (by ref)"=Make Fortnite Weapon Off Hand Melee Anim Trails Item Data!(
     AnimTrails:Particle System Soft Object Reference
     AnimTrailsNiagara:Niagara System Soft Object Reference=Make Soft Reference(Input=@n11)
     NiagaraSkeletonDIVariableName:Name=User.SkeletalMesh
     AnimTrailsFirstSocketName:Name
     AnimTrailsSecondSocketName:Name
     AnimTrailsWidth:float=1
     bAttachAnimTrailsToWeapon:bool=true
    ) -> (FortItemComponentData_WeaponMeleeAnimTrails_OffHand:Fortnite Weapon Off Hand Melee Anim Trails Item Data Structure) {is=title K2Node_MakeStruct}
   ) -> (then)
   > Logging Text(Message for Log="LOCTEXT(\"Niagara Offhand AnimTrail VFX was set in WID file.\")", Error Code=0)
   > Return Node(Success ?=true)
   n12 = Cast To ParticleSystem!(execute=@n11.CastFailed, Object=@n10) -> (then, CastFailed, AsParticle System:Particle System Object Reference) {is=title K2Node_DynamicCast}
   n13 = Get EUW_WID_FX_DualWieldWID_Setup()
   > SetAnimTrailDataOffhand(
    self=Get Dual Wield Weapon Item Definition(self=Get EUW_WID_FX_DualWieldWID_Setup())
    InAnimTrailData=Make Fortnite Weapon Off Hand Melee Anim Trails Item Data(
     AnimTrails=Make Soft Reference(Input:Particle System Object Reference=@n12) -> (Output:Particle System Soft Object Reference)
     NiagaraSkeletonDIVariableName=User.SkeletalMesh
     AnimTrailsFirstSocketName=Get Offhand Anim Trails First Socket Name!(self:EUW WID FX Dual Wield WID Setup Object Reference=@n13) -> (Offhand Anim Trails First Socket Name:Name=None)
     AnimTrailsSecondSocketName=Get Offhand Anim Trails Second Socket Name!(self:EUW WID FX Dual Wield WID Setup Object Reference=@n13) -> (Offhand Anim Trails Second Socket Name:Name=None)
     AnimTrailsWidth=Get Offhand Anim Trails Width!(self:EUW WID FX Dual Wield WID Setup Object Reference=@n13) -> (Offhand Anim Trails Width:double)
     bAttachAnimTrailsToWeapon=Get Offhand Attach Anim Trails to Weapon!(self:EUW WID FX Dual Wield WID Setup Object Reference=@n13) -> (Offhand Attach Anim Trails to Weapon:bool)
    )
   )
   > Logging Text(Message for Log="LOCTEXT(\"Cascade Offhand AnimTrail VFX was set in WID file.\")", Error Code=0)
   > Return Node(Success ?=true)
   Logging Text(
    execute=@n12.CastFailed
    Message for Log="LOCTEXT(\"ERROR - Offhand Anim Trail VFX was Valid, but it failed checks for Niagara and Cascade Systems, please check and confirm.\")"
    Error Code=2
   )
   > Large Warning Setup(Error Code=2) {is=title K2Node_CallFunction}
   > Return Node(Success ?=false)
   SetAnimTrailDataOffhand(
    execute=@n9.else
    self=Get Dual Wield Weapon Item Definition(self=Get EUW_WID_FX_DualWieldWID_Setup())
    InAnimTrailData=Make Fortnite Weapon Off Hand Melee Anim Trails Item Data(
     NiagaraSkeletonDIVariableName=User.SkeletalMesh
     AnimTrailsFirstSocketName=Melee_R_Upper
     AnimTrailsSecondSocketName=Melee_R_Lower
     AnimTrailsWidth=1
     bAttachAnimTrailsToWeapon=False
    )
   )
   > Logging Text(
    Message for Log="LOCTEXT(\"Offhand Anim Trail VFX was invalid.  This may be correct, so no Offhand Anim Trail effect was added to WID file and if one was in the WID, it was removed.\")"
    Error Code=0
   )
   > Return Node(Success ?=true)
   n14 = Cast To ParticleSystem(execute=@n6.CastFailed, Object=@n5)
   n15 = Get EUW_WID_FX_DualWieldWID_Setup()
   > SetAnimTrailData(
    self=Get Dual Wield Weapon Item Definition(self=Get EUW_WID_FX_DualWieldWID_Setup())
    InAnimTrailData=Make Fortnite Weapon Main Hand Melee Anim Trails Item Data(
     AnimTrails=Make Soft Reference(Input=@n14)
     NiagaraSkeletonDIVariableName=User.SkeletalMesh
     AnimTrailsFirstSocketName=Get Anim Trails First Socket Name!(self:EUW WID FX Dual Wield WID Setup Object Reference=@n15) -> (Anim Trails First Socket Name:Name=None)
     AnimTrailsSecondSocketName=Get Anim Trails Second Socket Name!(self:EUW WID FX Dual Wield WID Setup Object Reference=@n15) -> (Anim Trails Second Socket Name:Name=None)
     AnimTrailsWidth=Get Anim Trails Width!(self:EUW WID FX Dual Wield WID Setup Object Reference=@n15) -> (Anim Trails Width:double)
     bAttachAnimTrailsToWeapon=Get Attach Anim Trails to Weapon!(self:EUW WID FX Dual Wield WID Setup Object Reference=@n15) -> (Attach Anim Trails to Weapon:bool)
    )
   )
   > n16 = Logging Text(Message for Log="LOCTEXT(\"Cascade Anim Trail VFX was set in WID file.\")", Error Code=0)
   Logging Text(
    execute=@n14.CastFailed
    Message for Log="LOCTEXT(\"ERROR - Anim Trail VFX was Valid, but it failed checks for Niagara and Cascade Systems, please check and confirm.\")"
    Error Code=2
   )
   > Large Warning Setup(Error Code=2)
   > Return Node(Success ?=false)
   SetAnimTrailData(
    execute=@n4.else
    self=Get Dual Wield Weapon Item Definition(self=Get EUW_WID_FX_DualWieldWID_Setup())
    InAnimTrailData=Make Fortnite Weapon Main Hand Melee Anim Trails Item Data(
     NiagaraSkeletonDIVariableName=User.SkeletalMesh
     AnimTrailsFirstSocketName=Melee_R_Upper
     AnimTrailsSecondSocketName=Melee_R_Lower
     AnimTrailsWidth=1
     bAttachAnimTrailsToWeapon=False
    )
   )
   > n17 = Logging Text(
    Message for Log="LOCTEXT(\"Anim Trail VFX was invalid.  This may be correct, so no Anim Trail effect was added to WID file and if one was in the WID, it was removed.\")"
    Error Code=0
   )
   n18 = Get EUW_WID_FX_WID_Setup()
   n19 = Branch(
    execute=@n2.else
    Condition=IsValidSoftObjectReference(SoftObjectReference=Get AnimTrail VFX(self:EUW WID FX WID Setup Object Reference=@n18))
   )
   > n20 = LoadAsset_Blocking(Asset=Get AnimTrail VFX(self=Get EUW_WID_FX_WID_Setup()))
   > n21 = Cast To NiagaraSystem(Object=@n20)
   > SetAnimTrailData(
    self=Get Weapon Item Definition(self=Get EUW_WID_FX_WID_Setup())
    InAnimTrailData=Make Fortnite Weapon Main Hand Melee Anim Trails Item Data(
     AnimTrailsNiagara=Make Soft Reference(Input:Niagara System Object Reference=@n21) -> (Output:Niagara System Soft Object Reference)
     NiagaraSkeletonDIVariableName=User.SkeletalMesh
     AnimTrailsWidth=1
     bAttachAnimTrailsToWeapon=true
    )
   )
   > Logging Text(Message for Log="LOCTEXT(\"Niagara AnimTrail VFX was set in WID file.\")", Error Code=0)
   > Return Node(Success ?=true)
   n22 = Cast To ParticleSystem(execute=@n21.CastFailed, Object=@n20)
   n23 = Get EUW_WID_FX_WID_Setup()
   > SetAnimTrailData(
    self=Get Weapon Item Definition(self=Get EUW_WID_FX_WID_Setup())
    InAnimTrailData=Make Fortnite Weapon Main Hand Melee Anim Trails Item Data(
     AnimTrails=Make Soft Reference(Input:Particle System Object Reference=@n22) -> (Output:Particle System Soft Object Reference)
     NiagaraSkeletonDIVariableName=User.SkeletalMesh
     AnimTrailsFirstSocketName=Get Anim Trails First Socket Name(self:EUW WID FX WID Setup Object Reference=@n23) -> (Anim Trails First Socket Name=None)
     AnimTrailsSecondSocketName=Get Anim Trails Second Socket Name(self:EUW WID FX WID Setup Object Reference=@n23) -> (Anim Trails Second Socket Name=None)
     AnimTrailsWidth=Get Anim Trails Width(self:EUW WID FX WID Setup Object Reference=@n23)
     bAttachAnimTrailsToWeapon=Get Attach Anim Trails to Weapon(self:EUW WID FX WID Setup Object Reference=@n23)
    )
   )
   > Logging Text(Message for Log="LOCTEXT(\"Cascade Anim Trail VFX was set in WID file.\")", Error Code=0)
   > Return Node(Success ?=true)
   Logging Text(
    execute=@n22.CastFailed
    Message for Log="LOCTEXT(\"ERROR - AnimTrail VFX was Valid, but it failed checks for Niagara and Cascade Systems, please check and confirm.\")"
    Error Code=2
   )
   > Large Warning Setup(Error Code=2)
   > Return Node(Success ?=false)
   SetAnimTrailData(
    execute=@n19.else
    self=Get Weapon Item Definition(self=Get EUW_WID_FX_WID_Setup())
    InAnimTrailData=Make Fortnite Weapon Main Hand Melee Anim Trails Item Data(
     NiagaraSkeletonDIVariableName=User.SkeletalMesh
     AnimTrailsFirstSocketName=Melee_R_Upper
     AnimTrailsSecondSocketName=Melee_R_Lower
     AnimTrailsWidth=1
     bAttachAnimTrailsToWeapon=False
    )
   )
   > Logging Text(
    Message for Log="LOCTEXT(\"AnimTrail VFX was invalid.  This may be correct, so no Anim Trail effect was added to WID file and if one was in the WID, it was removed.\")"
    Error Code=0
   )
   > Return Node(Success ?=true)
  ]}
 Set Idle Item Data {nodes [
   n1 = Set Idle Item Data!() -> (then, Dual Wield ?:bool=~) {is=title K2Node_FunctionEntry}
   > n2 = Branch(Condition=@n1)
   n3 = Get EUW_WID_FX_DualWieldWID_Setup()
   > n4 = Branch(Condition=IsValidSoftObjectReference(SoftObjectReference=Get Idle VFX!(self:EUW WID FX Dual Wield WID Setup Object Reference=@n3) -> (Idle VFX:FXSystem Asset Soft Object Reference)))
   > n5 = LoadAsset_Blocking(Asset=Get Idle VFX(self=Get EUW_WID_FX_DualWieldWID_Setup()))
   > n6 = Cast To NiagaraSystem(Object=@n5)
   n7 = Get EUW_WID_FX_DualWieldWID_Setup()
   > SetIdleEffectData!(
    execute
    self:Fort Weapon Melee Item Definition Object Reference=Get Dual Wield Weapon Item Definition(self=Get EUW_WID_FX_DualWieldWID_Setup())
    InIdleEffectData:"Fortnite Weapon Main Hand Idle Effect Item Data Structure (by ref)"=Make Fortnite Weapon Main Hand Idle Effect Item Data!(
     IdleEffect:Particle System Soft Object Reference
     IdleEffectNiagara:Niagara System Soft Object Reference=Make Soft Reference(Input:Niagara System Object Reference=@n6) -> (Output:Niagara System Soft Object Reference)
     IdleFXSocketName:Name=Get Idle VFX Socket Name!(self:EUW WID FX Dual Wield WID Setup Object Reference=@n7) -> (Idle VFX Socket Name:Name=None)
    ) -> (FortItemComponentData_WeaponIdleEffect_MainHand:Fortnite Weapon Main Hand Idle Effect Item Data Structure) {is=title K2Node_MakeStruct}
   ) -> (then)
   > n8 = Logging Text(Message for Log="LOCTEXT(\"Niagara Idle VFX was set in WID file.\")", Error Code=0)
   n9 = Get EUW_WID_FX_DualWieldWID_Setup()
   n10 = Branch(
    execute=[@n8.then, @n18.then, @n19.then]
    Condition=IsValidSoftObjectReference(SoftObjectReference=Get Offhand Idle VFX!(self:EUW WID FX Dual Wield WID Setup Object Reference=@n9) -> (Offhand Idle VFX:FXSystem Asset Soft Object Reference))
   )
   > n11 = LoadAsset_Blocking(Asset=Get Offhand Idle VFX(self=Get EUW_WID_FX_DualWieldWID_Setup()))
   > n12 = Cast To NiagaraSystem(Object=@n11)
   n13 = Get EUW_WID_FX_DualWieldWID_Setup()
   > SetIdleEffectDataOffhand!(
    execute
    self:Fort Weapon Melee Dual Wield Item Definition Object Reference=Get Dual Wield Weapon Item Definition(self=Get EUW_WID_FX_DualWieldWID_Setup())
    InIdleEffectData:"Fortnite Weapon Off Hand Idle Effect Item Data Structure (by ref)"=Make Fortnite Weapon Off Hand Idle Effect Item Data!(
     IdleEffect:Particle System Soft Object Reference
     IdleEffectNiagara:Niagara System Soft Object Reference=Make Soft Reference(Input=@n12)
     IdleFXSocketName:Name=Get Offhand Idle VFX Socket Name!(self:EUW WID FX Dual Wield WID Setup Object Reference=@n13) -> (Offhand Idle VFX Socket Name:Name=None)
    ) -> (FortItemComponentData_WeaponIdleEffect_OffHand:Fortnite Weapon Off Hand Idle Effect Item Data Structure) {is=title K2Node_MakeStruct}
   ) -> (then)
   > Logging Text(Message for Log="LOCTEXT(\"Niagara Offhand Idle VFX was set in WID file.\")", Error Code=0)
   > Return Node(Success ?=true)
   n14 = Cast To ParticleSystem(execute=@n12.CastFailed, Object=@n11)
   n15 = Get EUW_WID_FX_DualWieldWID_Setup()
   > SetIdleEffectDataOffhand(
    self=Get Dual Wield Weapon Item Definition(self=Get EUW_WID_FX_DualWieldWID_Setup())
    InIdleEffectData=Make Fortnite Weapon Off Hand Idle Effect Item Data(
     IdleEffect=Make Soft Reference(Input:Particle System Object Reference=@n14) -> (Output:Particle System Soft Object Reference)
     IdleFXSocketName=Get Offhand Idle VFX Socket Name(self=@n15) -> (Offhand Idle VFX Socket Name=None)
    )
   )
   > Logging Text(Message for Log="LOCTEXT(\"Cascade Offhand Idle VFX was set in WID file.\")", Error Code=0)
   > Return Node(Success ?=true)
   Logging Text(
    execute=@n14.CastFailed
    Message for Log="LOCTEXT(\"ERROR - Offhand Idle VFX was Valid, but it failed checks for Niagara and Cascade Systems, please check and confirm.\")"
    Error Code=2
   )
   > Large Warning Setup(Error Code=2)
   > Return Node(Success ?=false)
   SetIdleEffectDataOffhand(
    execute=@n10.else
    self=Get Dual Wield Weapon Item Definition(self=Get EUW_WID_FX_DualWieldWID_Setup())
    InIdleEffectData=Make Fortnite Weapon Off Hand Idle Effect Item Data()
   )
   > Logging Text(
    Message for Log="LOCTEXT(\"Offhand Idle VFX was invalid.  This may be correct, so no Offhand Idle effect was added to WID file and if one was in the WID, it was removed.\")"
    Error Code=0
   )
   > Return Node(Success ?=true)
   n16 = Cast To ParticleSystem(execute=@n6.CastFailed, Object=@n5)
   n17 = Get EUW_WID_FX_DualWieldWID_Setup()
   > SetIdleEffectData(
    self=Get Dual Wield Weapon Item Definition(self=Get EUW_WID_FX_DualWieldWID_Setup())
    InIdleEffectData=Make Fortnite Weapon Main Hand Idle Effect Item Data(IdleEffect=Make Soft Reference(Input=@n16), IdleFXSocketName=Get Idle VFX Socket Name(self=@n17) -> (Idle VFX Socket Name=None))
   )
   > n18 = Logging Text(Message for Log="LOCTEXT(\"Cascade Idle VFX was set in WID file.\")", Error Code=0)
   Logging Text(
    execute=@n16.CastFailed
    Message for Log="LOCTEXT(\"ERROR - Idle VFX was Valid, but it failed checks for Niagara and Cascade Systems, please check and confirm.\")"
    Error Code=2
   )
   > Large Warning Setup(Error Code=2)
   > Return Node(Success ?=false)
   SetIdleEffectData(
    execute=@n4.else
    self=Get Dual Wield Weapon Item Definition(self=Get EUW_WID_FX_DualWieldWID_Setup())
    InIdleEffectData=Make Fortnite Weapon Main Hand Idle Effect Item Data()
   )
   > n19 = Logging Text(
    Message for Log="LOCTEXT(\"Idle VFX was invalid.  This may be correct, so no Idle effect was added to WID file and if one was in the WID, it was removed.\")"
    Error Code=0
   )
   n20 = Get EUW_WID_FX_WID_Setup()
   n21 = Branch(execute=@n2.else, Condition=IsValidSoftObjectReference(SoftObjectReference=Get Idle VFX(self:EUW WID FX WID Setup Object Reference=@n20)))
   > n22 = LoadAsset_Blocking(Asset=Get Idle VFX(self=Get EUW_WID_FX_WID_Setup()))
   > n23 = Cast To NiagaraSystem(Object=@n22)
   n24 = Get EUW_WID_FX_WID_Setup()
   > SetIdleEffectData(
    self=Get Weapon Item Definition(self=Get EUW_WID_FX_WID_Setup())
    InIdleEffectData=Make Fortnite Weapon Main Hand Idle Effect Item Data(
     IdleEffectNiagara=Make Soft Reference(Input:Niagara System Object Reference=@n23) -> (Output:Niagara System Soft Object Reference)
     IdleFXSocketName=Get Idle VFX Socket Name(self:EUW WID FX WID Setup Object Reference=@n24) -> (Idle VFX Socket Name=None)
    )
   )
   > Logging Text(Message for Log="LOCTEXT(\"Niagara Idle VFX was set in WID file.\")", Error Code=0)
   > Return Node(Success ?=true)
   n25 = Cast To ParticleSystem(execute=@n23.CastFailed, Object=@n22)
   n26 = Get EUW_WID_FX_WID_Setup()
   > SetIdleEffectData(
    self=Get Weapon Item Definition(self=Get EUW_WID_FX_WID_Setup())
    InIdleEffectData=Make Fortnite Weapon Main Hand Idle Effect Item Data(
     IdleEffect=Make Soft Reference(Input:Particle System Object Reference=@n25) -> (Output:Particle System Soft Object Reference)
     IdleFXSocketName=Get Idle VFX Socket Name(self=@n26) -> (Idle VFX Socket Name=None)
    )
   )
   > Logging Text(Message for Log="LOCTEXT(\"Cascade Idle VFX was set in WID file.\")", Error Code=0)
   > Return Node(Success ?=true)
   Logging Text(
    execute=@n25.CastFailed
    Message for Log="LOCTEXT(\"ERROR - Idle VFX was Valid, but it failed checks for Niagara and Cascade Systems, please check and confirm.\")"
    Error Code=2
   )
   > Large Warning Setup(Error Code=2)
   > Return Node(Success ?=false)
   SetIdleEffectData(
    execute=@n21.else
    self=Get Weapon Item Definition(self=Get EUW_WID_FX_WID_Setup())
    InIdleEffectData=Make Fortnite Weapon Main Hand Idle Effect Item Data()
   )
   > Logging Text(
    Message for Log="LOCTEXT(\"Idle VFX was invalid.  This may be correct, so no Idle effect was added to WID file and if one was in the WID, it was removed.\")"
    Error Code=0
   )
   > Return Node(Success ?=true)
  ]}
 Set Impact Item Data {nodes [
   n1 = Set Impact Item Data!() -> (then, Dual Wield ?:bool=~) {is=title K2Node_FunctionEntry}
   > n2 = Branch(Condition=@n1)
   n3 = Get EUW_WID_FX_DualWieldWID_Setup()
   > n4 = Branch(Condition=IsValidSoftObjectReference(SoftObjectReference=Get Impact VFX_Default(self:EUW WID FX Dual Wield WID Setup Object Reference=@n3)))
   > n5 = LoadAsset_Blocking(Asset=Get Impact VFX_Default(self=Get EUW_WID_FX_DualWieldWID_Setup()))
   > n6 = Cast To NiagaraSystem(Object=@n5)
   > n7 = Add!(
    execute
    TargetMap:Map of EPhysicalSurface Enums to Niagara System Soft Object References=Get Impact Niagara Physical Surface Effects Map!(self:EUW WID FX Dual Wield WID Setup Object Reference=Get EUW_WID_FX_DualWieldWID_Setup()) -> (Impact Niagara Physical Surface Effects Map:Map of EPhysicalSurface Enums to Niagara System Soft Object References=SurfaceType_Default)
    Key:"EPhysicalSurface Enum (by ref)"=SurfaceType_Default
    Value:"Niagara System Soft Object Reference (by ref)"=Make Soft Reference(Input:Niagara System Object Reference=@n6) -> (Output:Niagara System Soft Object Reference)
   ) -> (then) {is=title K2Node_CallFunction}
   n8 = Get EUW_WID_FX_DualWieldWID_Setup()
   SetImpactEffectData!(
    execute=[@n7.then, @n18.then]
    self:Fort Weapon Melee Item Definition Object Reference=Get Dual Wield Weapon Item Definition(self=@n8)
    InImpactEffectData:"Fortnite Weapon Main Hand Impact Effects Item Data Structure (by ref)"=Make Fortnite Weapon Main Hand Impact Effects Item Data!(
     ImpactPhysicalSurfaceEffectsMap:Map of EPhysicalSurface Enums to Particle System Soft Object References=Get Impact Cascade Physical Surface Effects Map!(self:EUW WID FX Dual Wield WID Setup Object Reference=@n8) -> (Impact Cascade Physical Surface Effects Map:Map of EPhysicalSurface Enums to Particle System Soft Object References=SurfaceType_Default)
     ImpactNiagaraPhysicalSurfaceEffectsMap:Map of EPhysicalSurface Enums to Niagara System Soft Object References=Get Impact Niagara Physical Surface Effects Map(self=@n8) -> (Impact Niagara Physical Surface Effects Map=SurfaceType_Default)
    ) -> (FortItemComponentData_WeaponImpactEffects_MainHand:Fortnite Weapon Main Hand Impact Effects Item Data Structure) {is=title K2Node_MakeStruct}
   ) -> (then)
   > Logging Text(Message for Log="LOCTEXT(\"Impact VFXs were set in WID file.\")", Error Code=0)
   n9 = Get EUW_WID_FX_DualWieldWID_Setup()
   > n10 = Branch(Condition=IsValidSoftObjectReference(SoftObjectReference=Get Offhand Impact VFX_Default(self=@n9)))
   > n11 = LoadAsset_Blocking(Asset=Get Offhand Impact VFX_Default(self=Get EUW_WID_FX_DualWieldWID_Setup()))
   > n12 = Cast To NiagaraSystem(Object=@n11)
   > n13 = Add(
    TargetMap=Get Offhand Impact Niagara Physical Surface Effects Map!(self:EUW WID FX Dual Wield WID Setup Object Reference=Get EUW_WID_FX_DualWieldWID_Setup()) -> (Offhand Impact Niagara Physical Surface Effects Map:Map of EPhysicalSurface Enums to Niagara System Soft Object References=SurfaceType_Default)
    Key=SurfaceType_Default
    Value=Make Soft Reference(Input=@n12)
   )
   n14 = Get EUW_WID_FX_DualWieldWID_Setup()
   SetImpactEffectDataOffhand!(
    execute=[@n13.then, @n16.then]
    self:Fort Weapon Melee Dual Wield Item Definition Object Reference=Get Dual Wield Weapon Item Definition(self=@n14)
    InImpactEffectData:"Fortnite Weapon Off Hand Impact Effects Item Data Structure (by ref)"=Make Fortnite Weapon Off Hand Impact Effects Item Data!(
     ImpactPhysicalSurfaceEffectsMap:Map of EPhysicalSurface Enums to Particle System Soft Object References=Get Offhand Impact Cascade Physical Surface Effects Map!(self:EUW WID FX Dual Wield WID Setup Object Reference=@n14) -> (Offhand Impact Cascade Physical Surface Effects Map:Map of EPhysicalSurface Enums to Particle System Soft Object References=SurfaceType_Default)
     ImpactNiagaraPhysicalSurfaceEffectsMap:Map of EPhysicalSurface Enums to Niagara System Soft Object References=Get Offhand Impact Niagara Physical Surface Effects Map(self=@n14) -> (Offhand Impact Niagara Physical Surface Effects Map=SurfaceType_Default)
    ) -> (FortItemComponentData_WeaponImpactEffects_OffHand:Fortnite Weapon Off Hand Impact Effects Item Data Structure) {is=title K2Node_MakeStruct}
   ) -> (then)
   > Logging Text(Message for Log="LOCTEXT(\"Offhand Impact VFXs were set in WID file.\")", Error Code=0)
   > Return Node(Success ?=true)
   n15 = Cast To ParticleSystem(execute=@n12.CastFailed, Object=@n11)
   > n16 = Add(
    TargetMap:Map of EPhysicalSurface Enums to Particle System Soft Object References=Get Offhand Impact Cascade Physical Surface Effects Map(self=Get EUW_WID_FX_DualWieldWID_Setup()) -> (Offhand Impact Cascade Physical Surface Effects Map=SurfaceType_Default)
    Key=SurfaceType_Default
    Value:"Particle System Soft Object Reference (by ref)"=Make Soft Reference(Input:Particle System Object Reference=@n15) -> (Output:Particle System Soft Object Reference)
   )
   Logging Text(
    execute=@n15.CastFailed
    Message for Log="LOCTEXT(\"Offhand Impact VFX was an invalid FX System Type.  Please add a Offhand Default Impact VFX to continue.\")"
    Error Code=2
   )
   > Large Warning Setup(Error Code=2)
   > Return Node(Success ?=false)
   Logging Text(
    execute=@n10.else
    Message for Log="LOCTEXT(\"Offhand Impact VFX was invalid.  Please add a Default Impact VFX to continue.\")"
    Error Code=2
   )
   > Large Warning Setup(Error Code=2)
   > Return Node(Success ?=false)
   n17 = Cast To ParticleSystem(execute=@n6.CastFailed, Object=@n5)
   > n18 = Add(
    TargetMap=Get Impact Cascade Physical Surface Effects Map(self=Get EUW_WID_FX_DualWieldWID_Setup()) -> (Impact Cascade Physical Surface Effects Map=SurfaceType_Default)
    Key=SurfaceType_Default
    Value=Make Soft Reference(Input=@n17)
   )
   Logging Text(
    execute=@n17.CastFailed
    Message for Log="LOCTEXT(\"Impact VFX was an invalid FX System Type.  Please add a Default Impact VFX to continue.\")"
    Error Code=2
   )
   > Large Warning Setup(Error Code=2)
   > Return Node(Success ?=false)
   Logging Text(execute=@n4.else, Message for Log="LOCTEXT(\"Impact VFX was invalid.  Please add a Default Impact VFX to continue.\")", Error Code=2)
   > Large Warning Setup(Error Code=2)
   > Return Node(Success ?=false)
   n19 = Get EUW_WID_FX_WID_Setup()
   n20 = Branch(
    execute=@n2.else
    Condition=IsValidSoftObjectReference(SoftObjectReference=Get Impact VFX_Default(self:EUW WID FX WID Setup Object Reference=@n19))
   )
   > n21 = LoadAsset_Blocking(Asset=Get Impact VFX_Default(self=Get EUW_WID_FX_WID_Setup()))
   > n22 = Cast To NiagaraSystem(Object=@n21)
   > n23 = Add(
    TargetMap:Map of EPhysicalSurface Enums to Niagara System Soft Object References=Get Impact Niagara Physical Surface Effects Map(self:EUW WID FX WID Setup Object Reference=Get EUW_WID_FX_WID_Setup()) -> (Impact Niagara Physical Surface Effects Map=SurfaceType_Default)
    Key=SurfaceType_Default
    Value:"Niagara System Soft Object Reference (by ref)"=Make Soft Reference(Input:Niagara System Object Reference=@n22) -> (Output:Niagara System Soft Object Reference)
   )
   n24 = Get EUW_WID_FX_WID_Setup()
   SetImpactEffectData(
    execute=[@n23.then, @n26.then]
    self=Get Weapon Item Definition(self=@n24)
    InImpactEffectData=Make Fortnite Weapon Main Hand Impact Effects Item Data(
     ImpactPhysicalSurfaceEffectsMap=Get Impact Cascade Physical Surface Effects Map(self:EUW WID FX WID Setup Object Reference=@n24) -> (Impact Cascade Physical Surface Effects Map=SurfaceType_Default)
     ImpactNiagaraPhysicalSurfaceEffectsMap=Get Impact Niagara Physical Surface Effects Map(self=@n24) -> (Impact Niagara Physical Surface Effects Map=SurfaceType_Default)
    )
   )
   > Logging Text(Message for Log="LOCTEXT(\"Impact VFXs were set in WID file.\")", Error Code=0)
   > Return Node(Success ?=true)
   n25 = Cast To ParticleSystem(execute=@n22.CastFailed, Object=@n21)
   > n26 = Add(
    TargetMap:Map of EPhysicalSurface Enums to Particle System Soft Object References=Get Impact Cascade Physical Surface Effects Map(self=Get EUW_WID_FX_WID_Setup()) -> (Impact Cascade Physical Surface Effects Map=SurfaceType_Default)
    Key=SurfaceType_Default
    Value:"Particle System Soft Object Reference (by ref)"=Make Soft Reference(Input:Particle System Object Reference=@n25) -> (Output:Particle System Soft Object Reference)
   )
   Logging Text(
    execute=@n25.CastFailed
    Message for Log="LOCTEXT(\"Impact VFX was an invalid FX System Type.  Please add a Default Impact VFX to continue.\")"
    Error Code=2
   )
   > Large Warning Setup(Error Code=2)
   > Return Node(Success ?=false)
   Logging Text(execute=@n20.else, Message for Log="LOCTEXT(\"Impact VFX was invalid.  Please add a Default Impact VFX to continue.\")", Error Code=2)
   > Large Warning Setup(Error Code=2)
   > Return Node(Success ?=false)
  ]}
 Set Pickaxe Actor Data {nodes [
   n1 = Set Pickaxe Actor Data!() -> (then, Dual Wield ?:bool=~) {is=title K2Node_FunctionEntry}
   > n2 = Branch(Condition=@n1)
   n3 = Resolve Soft Reference!(Input:Fort Weapon Soft Class Reference=Get Pickaxe Weapon Actor Class!(self:EUW WID FX Dual Wield WID Setup Object Reference=Get EUW_WID_FX_DualWieldWID_Setup()) -> (Pickaxe Weapon Actor Class:Fort Weapon Soft Class Reference)) -> (Output:Fort Weapon Class Reference)
   > n4 = Branch(Condition=IsValidClass!(Class:Object Class Reference=@n3) -> (ReturnValue:bool))
   > n5 = Cast To FortWeapon Class!(execute:Exec, Class:Object Class Reference=@n3) -> (then:Exec, CastFailed:Exec, AsFort Weapon:Fort Weapon Class Reference) {is=title K2Node_ClassDynamicCast}
   > SetWeaponActorClass!(
    execute
    self:Fort Weapon Item Definition Object Reference=Get Dual Wield Weapon Item Definition(self=Get EUW_WID_FX_DualWieldWID_Setup())
    InWeaponActorClass:Fort Weapon Class Reference=@n5
   ) -> (then)
   > Logging Text(Message for Log="LOCTEXT(\"Weapon Actor Class Set in WID\")", Error Code=0)
   > Return Node(Success ?=true)
   Logging Text(
    execute=@n5.CastFailed
    Message for Log="LOCTEXT(\"Pickaxe Actor Class was not a valid Weapon Actor Class.  Please correct and continue.\")"
    Error Code=2
   )
   > Large Warning Setup(Error Code=2)
   > Return Node(Success ?=false)
   Logging Text(execute=@n4.else, Message for Log="LOCTEXT(\"Pickaxe Actor Class was invalid Class.  Please correct and continue.\")", Error Code=2)
   > Large Warning Setup(Error Code=2)
   > Return Node(Success ?=false)
   n6 = Resolve Soft Reference(Input=Get Pickaxe Weapon Actor Class(self:EUW WID FX WID Setup Object Reference=Get EUW_WID_FX_WID_Setup()))
   n7 = Branch(execute=@n2.else, Condition=IsValidClass(Class=@n6))
   > n8 = Cast To FortWeapon Class(Class=@n6)
   > SetWeaponActorClass(self=Get Weapon Item Definition(self=Get EUW_WID_FX_WID_Setup()), InWeaponActorClass=@n8)
   > Logging Text(Message for Log="LOCTEXT(\"Weapon Actor Class Set in WID\")", Error Code=0)
   > Return Node(Success ?=true)
   Logging Text(
    execute=@n8.CastFailed
    Message for Log="LOCTEXT(\"Pickaxe Actor Class was not a valid Weapon Actor Class.  Please correct and continue.\")"
    Error Code=2
   )
   > Large Warning Setup(Error Code=2)
   > Return Node(Success ?=false)
   Logging Text(execute=@n7.else, Message for Log="LOCTEXT(\"Pickaxe Actor Class was invalid Class.  Please correct and continue.\")", Error Code=2)
   > Large Warning Setup(Error Code=2)
   > Return Node(Success ?=false)
  ]}
 Set Swing Item Data {nodes [
   n1 = Set Swing Item Data!() -> (then, Dual Wield ?:bool=~) {is=title K2Node_FunctionEntry}
   > n2 = Branch(Condition=@n1)
   n3 = Get EUW_WID_FX_DualWieldWID_Setup()
   > n4 = Branch(Condition=IsValidSoftObjectReference(SoftObjectReference=Get Swing VFX!(self:EUW WID FX Dual Wield WID Setup Object Reference=@n3) -> (Swing VFX:FXSystem Asset Soft Object Reference)))
   > n5 = LoadAsset_Blocking(Asset=Get Swing VFX(self=Get EUW_WID_FX_DualWieldWID_Setup()))
   > n6 = Cast To NiagaraSystem(Object=@n5)
   n7 = Get EUW_WID_FX_DualWieldWID_Setup()
   > SetSwingEffectData!(
    execute
    self:Fort Weapon Melee Item Definition Object Reference=Get Dual Wield Weapon Item Definition(self=Get EUW_WID_FX_DualWieldWID_Setup())
    InSwingEffectData:"Fortnite Weapon Main Hand Melee Swing Effect Item Data Structure (by ref)"=Make Fortnite Weapon Main Hand Melee Swing Effect Item Data!(
     SwingEffect:Particle System Soft Object Reference
     SwingEffectNiagara:Niagara System Soft Object Reference=Make Soft Reference(Input:Niagara System Object Reference=@n6) -> (Output:Niagara System Soft Object Reference)
     SwingFXSocketName:Name=Get Swing VFX Socket Name!(self:EUW WID FX Dual Wield WID Setup Object Reference=@n7) -> (Swing VFX Socket Name:Name=None)
    ) -> (FortItemComponentData_WeaponMeleeSwingEffect_MainHand:Fortnite Weapon Main Hand Melee Swing Effect Item Data Structure) {is=title K2Node_MakeStruct}
   ) -> (then)
   > n8 = Logging Text(Message for Log="LOCTEXT(\"Niagara Swing VFX was set in WID file.\")", Error Code=0)
   n9 = Get EUW_WID_FX_DualWieldWID_Setup()
   n10 = Branch(
    execute=[@n8.then, @n18.then, @n19.then]
    Condition=IsValidSoftObjectReference(SoftObjectReference=Get Offhand Swing VFX!(self:EUW WID FX Dual Wield WID Setup Object Reference=@n9) -> (Offhand Swing VFX:FXSystem Asset Soft Object Reference))
   )
   > n11 = LoadAsset_Blocking(Asset=Get Offhand Swing VFX(self=Get EUW_WID_FX_DualWieldWID_Setup()))
   > n12 = Cast To NiagaraSystem(Object=@n11)
   n13 = Get EUW_WID_FX_DualWieldWID_Setup()
   > SetSwingEffectDataOffhand!(
    execute
    self:Fort Weapon Melee Dual Wield Item Definition Object Reference=Get Dual Wield Weapon Item Definition(self=Get EUW_WID_FX_DualWieldWID_Setup())
    InSwingEffectData:"Fortnite Weapon Off Hand Melee Swing Effect Item Data Structure (by ref)"=Make Fortnite Weapon Off Hand Melee Swing Effect Item Data!(
     SwingEffect:Particle System Soft Object Reference
     SwingEffectNiagara:Niagara System Soft Object Reference=Make Soft Reference(Input=@n12)
     SwingFXSocketName:Name=Get Offhand Swing VFX Socket Name!(self:EUW WID FX Dual Wield WID Setup Object Reference=@n13) -> (Offhand Swing VFX Socket Name:Name=None)
    ) -> (FortItemComponentData_WeaponMeleeSwingEffect_OffHand:Fortnite Weapon Off Hand Melee Swing Effect Item Data Structure) {is=title K2Node_MakeStruct}
   ) -> (then)
   > Logging Text(Message for Log="LOCTEXT(\"Niagara Offhand Swing VFX was set in WID file.\")", Error Code=0)
   > Return Node(Success ?=true)
   n14 = Cast To ParticleSystem(execute=@n12.CastFailed, Object=@n11)
   n15 = Get EUW_WID_FX_DualWieldWID_Setup()
   > SetSwingEffectDataOffhand(
    self=Get Dual Wield Weapon Item Definition(self=Get EUW_WID_FX_DualWieldWID_Setup())
    InSwingEffectData=Make Fortnite Weapon Off Hand Melee Swing Effect Item Data(
     SwingEffect=Make Soft Reference(Input:Particle System Object Reference=@n14) -> (Output:Particle System Soft Object Reference)
     SwingFXSocketName=Get Offhand Swing VFX Socket Name(self=@n15) -> (Offhand Swing VFX Socket Name=None)
    )
   )
   > Logging Text(Message for Log="LOCTEXT(\"Cascade Offhand Swing VFX was set in WID file.\")", Error Code=0)
   > Return Node(Success ?=true)
   Logging Text(
    execute=@n14.CastFailed
    Message for Log="LOCTEXT(\"ERROR - Offhand Swing VFX was Valid, but it failed checks for Niagara and Cascade Systems, please check and confirm.\")"
    Error Code=2
   )
   > Large Warning Setup(Error Code=2)
   > Return Node(Success ?=false)
   SetSwingEffectDataOffhand(
    execute=@n10.else
    self=Get Dual Wield Weapon Item Definition(self=Get EUW_WID_FX_DualWieldWID_Setup())
    InSwingEffectData=Make Fortnite Weapon Off Hand Melee Swing Effect Item Data(SwingFXSocketName=SwingFX)
   )
   > Logging Text(
    Message for Log="LOCTEXT(\"Offhand Swing VFX was invalid.  This may be correct, so no Offhand Swing effect was added to WID file and if one was in the WID, it was removed.\")"
    Error Code=0
   )
   > Return Node(Success ?=true)
   n16 = Cast To ParticleSystem(execute=@n6.CastFailed, Object=@n5)
   n17 = Get EUW_WID_FX_DualWieldWID_Setup()
   > SetSwingEffectData(
    self=Get Dual Wield Weapon Item Definition(self=Get EUW_WID_FX_DualWieldWID_Setup())
    InSwingEffectData=Make Fortnite Weapon Main Hand Melee Swing Effect Item Data(SwingEffect=Make Soft Reference(Input=@n16), SwingFXSocketName=Get Swing VFX Socket Name(self=@n17) -> (Swing VFX Socket Name=None))
   )
   > n18 = Logging Text(Message for Log="LOCTEXT(\"Cascade Swing VFX was set in WID file.\")", Error Code=0)
   Logging Text(
    execute=@n16.CastFailed
    Message for Log="LOCTEXT(\"ERROR - Swing VFX was Valid, but it failed checks for Niagara and Cascade Systems, please check and confirm.\")"
    Error Code=2
   )
   > Large Warning Setup(Error Code=2)
   > Return Node(Success ?=false)
   SetSwingEffectData(
    execute=@n4.else
    self=Get Dual Wield Weapon Item Definition(self=Get EUW_WID_FX_DualWieldWID_Setup())
    InSwingEffectData=Make Fortnite Weapon Main Hand Melee Swing Effect Item Data()
   )
   > n19 = Logging Text(
    Message for Log="LOCTEXT(\"Swing VFX was invalid.  This may be correct, so no Swing effect was added to WID file and if one was in the WID, it was removed.\")"
    Error Code=0
   )
   n20 = Get EUW_WID_FX_WID_Setup()
   n21 = Branch(execute=@n2.else, Condition=IsValidSoftObjectReference(SoftObjectReference=Get Swing VFX(self:EUW WID FX WID Setup Object Reference=@n20)))
   > n22 = LoadAsset_Blocking(Asset=Get Swing VFX(self=Get EUW_WID_FX_WID_Setup()))
   > n23 = Cast To NiagaraSystem(Object=@n22)
   n24 = Get EUW_WID_FX_WID_Setup()
   > SetSwingEffectData(
    self=Get Weapon Item Definition(self=Get EUW_WID_FX_WID_Setup())
    InSwingEffectData=Make Fortnite Weapon Main Hand Melee Swing Effect Item Data(
     SwingEffectNiagara=Make Soft Reference(Input:Niagara System Object Reference=@n23) -> (Output:Niagara System Soft Object Reference)
     SwingFXSocketName=Get Swing VFX Socket Name(self:EUW WID FX WID Setup Object Reference=@n24) -> (Swing VFX Socket Name=None)
    )
   )
   > Logging Text(Message for Log="LOCTEXT(\"Niagara Swing VFX was set in WID file.\")", Error Code=0)
   > Return Node(Success ?=true)
   n25 = Cast To ParticleSystem(execute=@n23.CastFailed, Object=@n22)
   n26 = Get EUW_WID_FX_WID_Setup()
   > SetSwingEffectData(
    self=Get Weapon Item Definition(self=Get EUW_WID_FX_WID_Setup())
    InSwingEffectData=Make Fortnite Weapon Main Hand Melee Swing Effect Item Data(
     SwingEffect=Make Soft Reference(Input:Particle System Object Reference=@n25) -> (Output:Particle System Soft Object Reference)
     SwingFXSocketName=Get Swing VFX Socket Name(self=@n26) -> (Swing VFX Socket Name=None)
    )
   )
   > Logging Text(Message for Log="LOCTEXT(\"Cascade Swing VFX was set in WID file.\")", Error Code=0)
   > Return Node(Success ?=true)
   Logging Text(
    execute=@n25.CastFailed
    Message for Log="LOCTEXT(\"ERROR - Swing VFX was Valid, but it failed checks for Niagara and Cascade Systems, please check and confirm.\")"
    Error Code=2
   )
   > Large Warning Setup(Error Code=2)
   > Return Node(Success ?=false)
   SetSwingEffectData(
    execute=@n21.else
    self=Get Weapon Item Definition(self=Get EUW_WID_FX_WID_Setup())
    InSwingEffectData=Make Fortnite Weapon Main Hand Melee Swing Effect Item Data(SwingFXSocketName=SwingFX)
   )
   > Logging Text(
    Message for Log="LOCTEXT(\"Swing VFX was invalid.  This may be correct, so no Swing effect was added to WID file and if one was in the WID, it was removed.\")"
    Error Code=0
   )
   > Return Node(Success ?=true)
  ]}
 WID Setup {nodes [
   n1 = WID Setup!() -> (then, Dual Wield ?:bool=~) {is=title K2Node_FunctionEntry}
   > n2 = Branch(Condition=@n1)
   n3 = Get Pickaxe_ID()
   n4 = Get EUW_WID_FX_DualWieldWID_Setup()
   > Set Pickaxe ID!(execute, Pickaxe ID:Athena Pickaxe Item Definition Object Reference=@n3, self:EUW WID FX Dual Wield WID Setup Object Reference=@n4) -> (then, Output_Get:Athena Pickaxe Item Definition Object Reference) {is=title K2Node_VariableSet}
   > Set Pickaxe ID_Setter!(execute, Pickaxe ID_Setter:Athena Pickaxe Item Definition Object Reference=@n3, self:EUW WID FX Dual Wield WID Setup Object Reference=@n4) -> (then, Output_Get:Athena Pickaxe Item Definition Object Reference)
   > Set Dual Wield Weapon Item Definition!(
    execute
    Dual Wield Weapon Item Definition:Fort Weapon Melee Dual Wield Item Definition Object Reference=Cast To FortWeaponMeleeDualWieldItemDefinition!(Object=Get WeaponDefinition(self=@n3)) -> (AsFort Weapon Melee Dual Wield Item Definition:Fort Weapon Melee Dual Wield Item Definition Object Reference, bSuccess:bool=~) {is=title K2Node_DynamicCast}
    self:EUW WID FX Dual Wield WID Setup Object Reference=@n4
   ) -> (then, Output_Get:Fort Weapon Melee Dual Wield Item Definition Object Reference)
   n5 = Get Pickaxe_ID()
   n6 = Get EUW_WID_FX_WID_Setup()
   Set Pickaxe ID(execute=@n2.else, Pickaxe ID=@n5, self:EUW WID FX WID Setup Object Reference=@n6)
   > Set Pickaxe ID_Setter(Pickaxe ID_Setter=@n5, self:EUW WID FX WID Setup Object Reference=@n6)
   > Set Weapon Item Definition!(
    execute
    Weapon Item Definition:Fort Weapon Melee Item Definition Object Reference=Get WeaponDefinition(self=Get Pickaxe_ID())
    self:EUW WID FX WID Setup Object Reference=@n6
   ) -> (then, Output_Get:Fort Weapon Melee Item Definition Object Reference)
   Get WID() -> (WID:Fort Weapon Melee Item Definition Object Reference)
  ]}
 WIDValidationLogging {nodes [
   n1 = WIDValidationLogging!() -> (then, Dual Wield ?:bool=~) {is=title K2Node_FunctionEntry}
   > Clear Logging Text()
   > n2 = Branch(Condition=@n1)
   > n3 = Branch(Condition=Get Use Cascade Idle!(self:EUW WID FX Dual Wield WID Setup Object Reference=Get EUW_WID_FX_DualWieldWID_Setup()) -> (Use Cascade Idle:bool))
   > n4 = Logging Text(Message for Log="LOCTEXT(\"Cascade Idle VFX Found - Consider converting to a Niagara System.\")", Error Code=1)
   n5 = Branch(
    execute=[@n4.then, @n3.else]
    Condition=Get Use Cascade Idle_Offhand!(self:EUW WID FX Dual Wield WID Setup Object Reference=Get EUW_WID_FX_DualWieldWID_Setup()) -> (Use Cascade Idle_Offhand:bool)
   )
   > n6 = Logging Text(Message for Log="LOCTEXT(\"Cascade Offhand Idle VFX Found - Consider converting to a Niagara System.\")", Error Code=1)
   n7 = Branch(
    execute=[@n6.then, @n5.else]
    Condition=Get Use Cascade Swing!(self:EUW WID FX Dual Wield WID Setup Object Reference=Get EUW_WID_FX_DualWieldWID_Setup()) -> (Use Cascade Swing:bool)
   )
   > n8 = Logging Text(Message for Log="LOCTEXT(\"Cascade Swing VFX Found - Consider converting to a Niagara System.\")", Error Code=1)
   n9 = Branch(
    execute=[@n8.then, @n7.else]
    Condition=Get Use Cascade Swing__Offhand!(self:EUW WID FX Dual Wield WID Setup Object Reference=Get EUW_WID_FX_DualWieldWID_Setup()) -> (Use Cascade Swing__Offhand:bool)
   )
   > n10 = Logging Text(Message for Log="LOCTEXT(\"Cascade Offhand Swing VFX Found - Consider converting to a Niagara System.\")", Error Code=1)
   n11 = Branch(
    execute=[@n10.then, @n9.else]
    Condition=Get Use Cascade AnimTrail!(self:EUW WID FX Dual Wield WID Setup Object Reference=Get EUW_WID_FX_DualWieldWID_Setup()) -> (Use Cascade AnimTrail:bool)
   )
   > n12 = Logging Text(Message for Log="LOCTEXT(\"Cascade AnimTrail VFX Found - Consider converting to a Niagara System.\")", Error Code=1)
   n13 = Branch(
    execute=[@n12.then, @n11.else]
    Condition=Get Use Cascade AnimTrail__Offhand!(self:EUW WID FX Dual Wield WID Setup Object Reference=Get EUW_WID_FX_DualWieldWID_Setup()) -> (Use Cascade AnimTrail__Offhand:bool)
   )
   > n14 = Logging Text(Message for Log="LOCTEXT(\"Cascade Offhand AnimTrail VFX Found - Consider converting to a Niagara System.\")", Error Code=1)
   n15 = Branch(
    execute=[@n14.then, @n13.else]
    Condition=Get Use Cascade Impacts!(self:EUW WID FX Dual Wield WID Setup Object Reference=Get EUW_WID_FX_DualWieldWID_Setup()) -> (Use Cascade Impacts:bool)
   )
   > n16 = Logging Text(Message for Log="LOCTEXT(\"Cascade Impact VFX Found - Consider converting to a Niagara System.\")", Error Code=1)
   n17 = Branch(
    execute=[@n16.then, @n15.else]
    Condition=Get Use Cascade Impacts__Offhand!(self:EUW WID FX Dual Wield WID Setup Object Reference=Get EUW_WID_FX_DualWieldWID_Setup()) -> (Use Cascade Impacts__Offhand:bool)
   )
   > n18 = Logging Text(Message for Log="LOCTEXT(\"Cascade Offhand Impact VFX Found - Consider converting to a Niagara System.\")", Error Code=1)
   n19 = Branch(
    execute=[@n18.then, @n17.else]
    Condition=Get Missing Impacts!(self:EUW WID FX Dual Wield WID Setup Object Reference=Get EUW_WID_FX_DualWieldWID_Setup()) -> (Missing Impacts:bool)
   )
   > Logging Text(
    Message for Log="LOCTEXT(\"ERROR - NO Impact VFX Found - Pickaxe will use Blueprint Cascade Impact FX, Please assign an Impact VFX.\")"
    Error Code=2
   )
   > n20 = Branch(Condition=Get Missing Impacts__Offhand!(self:EUW WID FX Dual Wield WID Setup Object Reference=Get EUW_WID_FX_DualWieldWID_Setup()) -> (Missing Impacts__Offhand:bool))
   > Logging Text(
    Message for Log="LOCTEXT(\"WARNING - NO Impact Offhand VFX Found - This may be intentional if a 1H Pickaxe Animation is assigned.  If not, Pickaxe will use Blueprint Cascade Impact FX; if that\\'s the case, please assign an Impact VFX.\")"
    Error Code=1
   )
   > Large Warning Setup(Error Code=2)
   > Return Node!(execute)
   Large Warning Setup(execute=@n20.else, Error Code=2)
   > Return Node()
   n21 = Branch(execute=@n19.else, Condition=Get Missing Impacts__Offhand(self=Get EUW_WID_FX_DualWieldWID_Setup()))
   > Logging Text(
    Message for Log="LOCTEXT(\"WARNING - NO Impact Offhand VFX Found - This may be intentional if a 1H Pickaxe Animation is assigned.  If not, Pickaxe will use Blueprint Cascade Impact FX; if that\\'s the case, please assign an Impact VFX.\")"
    Error Code=1
   )
   > Large Warning Setup(Error Code=2)
   > Return Node()
   n22 = Get EUW_WID_FX_DualWieldWID_Setup()
   n23 = Get EUW_WID_FX_DualWieldWID_Setup()
   n24 = Get EUW_WID_FX_DualWieldWID_Setup()
   n25 = Get EUW_WID_FX_DualWieldWID_Setup()
   n26 = Get EUW_WID_FX_DualWieldWID_Setup()
   n27 = Get EUW_WID_FX_DualWieldWID_Setup()
   n28 = Get EUW_WID_FX_DualWieldWID_Setup()
   n29 = Get EUW_WID_FX_DualWieldWID_Setup()
   n30 = Branch(
    execute=@n21.else
    Condition=OR Boolean!(
     A=Get Use Cascade Idle(self=@n22)
     B=Get Use Cascade Idle_Offhand(self=@n25)
     C:bool=Get Use Cascade Swing(self=@n23)
     D:bool=Get Use Cascade Swing__Offhand(self=@n26)
     E:bool=Get Use Cascade AnimTrail(self=@n24)
     F:bool=Get Use Cascade AnimTrail__Offhand(self=@n27)
     G:bool=Get Use Cascade Impacts(self=@n28)
     H:bool=Get Use Cascade Impacts__Offhand(self=@n29)
    ) -> (ReturnValue)
   )
   > n31 = Large Warning Setup(Error Code=1)
   Return Node(execute=[@n32.then, @n31.then])
   n32 = Large Warning Setup(execute=@n30.else, Error Code=0)
   n33 = Branch(execute=@n2.else, Condition=Get Use Cascade Idle(self:EUW WID FX WID Setup Object Reference=Get EUW_WID_FX_WID_Setup()))
   > n34 = Logging Text(Message for Log="LOCTEXT(\"Cascade Idle VFX Found - Consider converting to a Niagara System.\")", Error Code=1)
   n35 = Branch(execute=[@n34.then, @n33.else], Condition=Get Use Cascade Swing(self:EUW WID FX WID Setup Object Reference=Get EUW_WID_FX_WID_Setup()))
   > n36 = Logging Text(Message for Log="LOCTEXT(\"Cascade Swing VFX Found - Consider converting to a Niagara System.\")", Error Code=1)
   n37 = Branch(execute=[@n36.then, @n35.else], Condition=Get Use Cascade AnimTrail(self:EUW WID FX WID Setup Object Reference=Get EUW_WID_FX_WID_Setup()))
   > n38 = Logging Text(Message for Log="LOCTEXT(\"Cascade AnimTrail VFX Found - Consider converting to a Niagara System.\")", Error Code=1)
   n39 = Branch(execute=[@n38.then, @n37.else], Condition=Get Use Cascade Impacts(self:EUW WID FX WID Setup Object Reference=Get EUW_WID_FX_WID_Setup()))
   > n40 = Logging Text(Message for Log="LOCTEXT(\"Cascade Impact VFX Found - Consider converting to a Niagara System.\")", Error Code=1)
   n41 = Branch(execute=[@n40.then, @n39.else], Condition=Get Missing Impacts(self:EUW WID FX WID Setup Object Reference=Get EUW_WID_FX_WID_Setup()))
   > Logging Text(
    Message for Log="LOCTEXT(\"ERROR - NO Impact VFX Found - Pickaxe will use Blueprint Cascade Impact FX, Please assign an Impact VFX.\")"
    Error Code=2
   )
   > Large Warning Setup(Error Code=2)
   > Return Node()
   n42 = Get EUW_WID_FX_WID_Setup()
   n43 = Get EUW_WID_FX_WID_Setup()
   n44 = Get EUW_WID_FX_WID_Setup()
   n45 = Get EUW_WID_FX_WID_Setup()
   n46 = Branch(
    execute=@n41.else
    Condition=OR Boolean!(
     A=Get Use Cascade Idle(self=@n44)
     B=Get Use Cascade Swing(self=@n42)
     C=Get Use Cascade AnimTrail(self=@n43)
     D=Get Use Cascade Impacts(self=@n45)
    ) -> (ReturnValue)
   )
   > n47 = Large Warning Setup(Error Code=1)
   Return Node(execute=[@n48.then, @n47.then])
   n48 = Large Warning Setup(execute=@n46.else, Error Code=0)
  ]}
]
variables [
 Allow Interaction {type=boolean}
 Impact Niagara Physical Surface Effects Map_BASE {default {
   SurfaceType1=null
   SurfaceType10=null
   SurfaceType11=null
   SurfaceType12=null
   SurfaceType13=null
   SurfaceType14=null
   SurfaceType15=null
   SurfaceType16=null
   SurfaceType18=null
   SurfaceType19=null
   SurfaceType2=null
   SurfaceType20=null
   SurfaceType21=null
   SurfaceType22=null
   SurfaceType23=null
   SurfaceType24=null
   SurfaceType25=null
   SurfaceType26=null
   SurfaceType27=null
   SurfaceType28=null
   SurfaceType29=null
   SurfaceType3=null
   SurfaceType30=null
   SurfaceType31=null
   SurfaceType32=null
   SurfaceType33=null
   SurfaceType34=null
   SurfaceType4=null
   SurfaceType5=null
   SurfaceType6=null
   SurfaceType7=null
   SurfaceType8=null
   SurfaceType9=null
   SurfaceType_Default=null
  }}
 Impact Physical Surface Effects Map_BASE {default {
   SurfaceType1=null
   SurfaceType10=null
   SurfaceType11=null
   SurfaceType12=null
   SurfaceType13=null
   SurfaceType14=null
   SurfaceType15=null
   SurfaceType16=null
   SurfaceType18=null
   SurfaceType19=null
   SurfaceType2=null
   SurfaceType20=null
   SurfaceType21=null
   SurfaceType22=null
   SurfaceType23=null
   SurfaceType24=null
   SurfaceType25=null
   SurfaceType26=null
   SurfaceType27=null
   SurfaceType28=null
   SurfaceType29=null
   SurfaceType3=null
   SurfaceType30=null
   SurfaceType31=null
   SurfaceType32=null
   SurfaceType33=null
   SurfaceType34=null
   SurfaceType4=null
   SurfaceType5=null
   SurfaceType6=null
   SurfaceType7=null
   SurfaceType8=null
   SurfaceType9=null
   SurfaceType_Default=null
  }}
 isDualWield? {type=boolean}
 Pickaxe_ID {type=AthenaPickaxeItemDefinition, default=null}
]
widgets {
 info {rootWidgetClass=&CanvasPanel, widgetCount=18}
 widgets [
  CanvasPanel_365 {widgetClassPath=&CanvasPanel, widget=CanvasPanel_365}
  Overlay_26 {
   widgetClassPath=&Overlay
   parent=CanvasPanel_365
   widget=Overlay_26
   slot {
    class=CanvasPanelSlot
    properties {bAutoSize=true, layoutData {alignment=<0 0>, anchors {maximum=<1 1>, minimum=<0 0>}, offsets=<ltrb 0 0 0 0>}}
   }
  }
  HorizontalBox_170 {
   widgetClassPath=&HorizontalBox
   parent=Overlay_26
   widget=HorizontalBox_170
   slot {class=OverlaySlot, properties {horizontalAlignment=HAlign_Fill, verticalAlignment=VAlign_Fill}}
  }
  WarningMessages {
   widgetClassPath=&HorizontalBox
   parent=HorizontalBox_170
   widget=WarningMessages
   slot {class=HorizontalBoxSlot, properties {horizontalAlignment=HAlign_Left, verticalAlignment=VAlign_Top}}
  }
  WarningIcon {
   widgetClassPath=&Image
   bIsVariable=true
   parent=WarningMessages
   widget=WarningIcon
   properties {
    visibility=Hidden
    brush {
     bIsDynamicallyLoaded=false
     drawAs=Image
     imageType=FullColor
     mirroring=NoMirror
     resourceName=None
     resourceObject=&T-Icon-Warning-128
     tiling=NoTile
     imageSize=<32 32>
     margin=<ltrb 0 0 0 0>
     outlineSettings {
      bUseBrushTransparency=false
      roundingType=HalfHeightRadius
      width=0
      color {colorUseRule=UseColor_Specified, specifiedColor=<rgba 0 0 0 0>}
      cornerRadii=<0 0 0 0>
     }
     tintColor {colorUseRule=UseColor_Specified, specifiedColor=<rgba 1 0.6128 0 1>}
     uVRegion {bIsValid=false, max=<0 0>, min=<0 0>}
    }
   }
   slot {class=HorizontalBoxSlot, properties {horizontalAlignment=HAlign_Left, verticalAlignment=VAlign_Top}}
  }
  LargeWarningText {
   widgetClassPath=&TextBlock
   bIsVariable=true
   parent=WarningMessages
   widget=LargeWarningText
   properties {
    text=WARNING
    visibility=Hidden
    colorAndOpacity {colorUseRule=UseColor_Specified, specifiedColor=<rgba 1 0.6128 0 1>}
    font {
     bForceMonospaced=false
     bMaterialIsStencil=false
     fontMaterial=null
     fontObject=&Burbank
     letterSpacing=0
     monospacedWidth=1
     size=18
     skewAmount=0
     typefaceFontName=Burbank-PlayerPower
     outlineSettings {
      bApplyOutlineToDropShadows=false
      bMiteredCorners=false
      bSeparateFillAlpha=false
      outlineMaterial=null
      outlineSize=0
      outlineColor=<rgba 0 0 0 1>
     }
    }
   }
   slot {class=HorizontalBoxSlot, properties {horizontalAlignment=HAlign_Center, verticalAlignment=VAlign_Center}}
  }
  Interact_Button {
   widgetClassPath=&EditorUtilityButton
   parent=HorizontalBox_170
   widget=Interact_Button
   slot {
    class=HorizontalBoxSlot
    properties {horizontalAlignment=HAlign_Right, verticalAlignment=VAlign_Top, padding=<ltrb 0 2 5 0>, size {sizeRule=Fill, value=1}}
   }
  }
  Validation {
   widgetClassPath=&TextBlock
   bIsVariable=true
   parent=Interact_Button
   widget=Validation
   properties {
    text=Validate Changes
    font {
     bForceMonospaced=false
     bMaterialIsStencil=false
     fontMaterial=null
     fontObject=&Burbank
     letterSpacing=0
     monospacedWidth=1
     size=15
     skewAmount=0
     typefaceFontName=BurbankRegular-Black
     outlineSettings {
      bApplyOutlineToDropShadows=false
      bMiteredCorners=false
      bSeparateFillAlpha=false
      outlineMaterial=null
      outlineSize=0
      outlineColor=<rgba 0 0 0 1>
     }
    }
   }
   slot {class=ButtonSlot, properties {padding=<ltrb 0 0 0 0>}}
  }
  CommonHierarchicalScrollBox_76 {
   widgetClassPath=/Script/CommonUI.CommonHierarchicalScrollBox
   parent=Overlay_26
   widget=CommonHierarchicalScrollBox_76
   properties {
    wheelScrollMultiplier=2
    scrollbarThickness=<5 5>
    widgetBarStyle {
     thickness=8
     draggedThumbImage {
      bIsDynamicallyLoaded=false
      drawAs=RoundedBox
      imageType=NoImage
      mirroring=NoMirror
      resourceName=None
      resourceObject=null
      tiling=NoTile
      imageSize=<0 0>
      margin=<ltrb 0 0 0 0>
      outlineSettings {
       bUseBrushTransparency=false
       roundingType=FixedRadius
       width=0
       color {colorUseRule=UseColor_Specified, specifiedColor=<rgba 0 0 0 0>}
       cornerRadii=<4 4 4 4>
      }
      tintColor {colorUseRule=UseColor_Specified, specifiedColor=<rgba 0.2159 0.2159 0.2159 1>}
      uVRegion {bIsValid=false, max=<0 0>, min=<0 0>}
     }
     horizontalBackgroundImage {
      bIsDynamicallyLoaded=false
      drawAs=NoDrawType
      imageType=NoImage
      mirroring=NoMirror
      resourceName=None
      resourceObject=null
      tiling=NoTile
      imageSize=<0 0>
      margin=<ltrb 0 0 0 0>
      outlineSettings {
       bUseBrushTransparency=false
       roundingType=HalfHeightRadius
       width=0
       color {colorUseRule=UseColor_Specified, specifiedColor=<rgba 0 0 0 0>}
       cornerRadii=<0 0 0 0>
      }
      tintColor {colorUseRule=UseColor_Specified, specifiedColor=<rgba 1 1 1 1>}
      uVRegion {bIsValid=false, max=<0 0>, min=<0 0>}
     }
     horizontalBottomSlotImage {
      bIsDynamicallyLoaded=false
      drawAs=NoDrawType
      imageType=NoImage
      mirroring=NoMirror
      resourceName=None
      resourceObject=null
      tiling=NoTile
      imageSize=<0 0>
      margin=<ltrb 0 0 0 0>
      outlineSettings {
       bUseBrushTransparency=false
       roundingType=HalfHeightRadius
       width=0
       color {colorUseRule=UseColor_Specified, specifiedColor=<rgba 0 0 0 0>}
       cornerRadii=<0 0 0 0>
      }
      tintColor {colorUseRule=UseColor_Specified, specifiedColor=<rgba 1 1 1 1>}
      uVRegion {bIsValid=false, max=<0 0>, min=<0 0>}
     }
     horizontalTopSlotImage {
      bIsDynamicallyLoaded=false
      drawAs=NoDrawType
      imageType=NoImage
      mirroring=NoMirror
      resourceName=None
      resourceObject=null
      tiling=NoTile
      imageSize=<0 0>
      margin=<ltrb 0 0 0 0>
      outlineSettings {
       bUseBrushTransparency=false
       roundingType=HalfHeightRadius
       width=0
       color {colorUseRule=UseColor_Specified, specifiedColor=<rgba 0 0 0 0>}
       cornerRadii=<0 0 0 0>
      }
      tintColor {colorUseRule=UseColor_Specified, specifiedColor=<rgba 1 1 1 1>}
      uVRegion {bIsValid=false, max=<0 0>, min=<0 0>}
     }
     hoveredThumbImage {
      bIsDynamicallyLoaded=false
      drawAs=RoundedBox
      imageType=NoImage
      mirroring=NoMirror
      resourceName=None
      resourceObject=null
      tiling=NoTile
      imageSize=<0 0>
      margin=<ltrb 0 0 0 0>
      outlineSettings {
       bUseBrushTransparency=false
       roundingType=FixedRadius
       width=0
       color {colorUseRule=UseColor_Specified, specifiedColor=<rgba 0 0 0 0>}
       cornerRadii=<4 4 4 4>
      }
      tintColor {colorUseRule=UseColor_Specified, specifiedColor=<rgba 0.2159 0.2159 0.2159 1>}
      uVRegion {bIsValid=false, max=<0 0>, min=<0 0>}
     }
     normalThumbImage {
      bIsDynamicallyLoaded=false
      drawAs=RoundedBox
      imageType=NoImage
      mirroring=NoMirror
      resourceName=None
      resourceObject=null
      tiling=NoTile
      imageSize=<0 0>
      margin=<ltrb 0 0 0 0>
      outlineSettings {
       bUseBrushTransparency=false
       roundingType=FixedRadius
       width=0
       color {colorUseRule=UseColor_Specified, specifiedColor=<rgba 0 0 0 0>}
       cornerRadii=<4 4 4 4>
      }
      tintColor {colorUseRule=UseColor_Specified, specifiedColor=<rgba 0.0953 0.0953 0.0953 1>}
      uVRegion {bIsValid=false, max=<0 0>, min=<0 0>}
     }
     verticalBackgroundImage {
      bIsDynamicallyLoaded=false
      drawAs=NoDrawType
      imageType=NoImage
      mirroring=NoMirror
      resourceName=None
      resourceObject=null
      tiling=NoTile
      imageSize=<0 0>
      margin=<ltrb 0 0 0 0>
      outlineSettings {
       bUseBrushTransparency=false
       roundingType=HalfHeightRadius
       width=0
       color {colorUseRule=UseColor_Specified, specifiedColor=<rgba 0 0 0 0>}
       cornerRadii=<0 0 0 0>
      }
      tintColor {colorUseRule=UseColor_Specified, specifiedColor=<rgba 1 1 1 1>}
      uVRegion {bIsValid=false, max=<0 0>, min=<0 0>}
     }
     verticalBottomSlotImage {
      bIsDynamicallyLoaded=false
      drawAs=NoDrawType
      imageType=NoImage
      mirroring=NoMirror
      resourceName=None
      resourceObject=null
      tiling=NoTile
      imageSize=<0 0>
      margin=<ltrb 0 0 0 0>
      outlineSettings {
       bUseBrushTransparency=false
       roundingType=HalfHeightRadius
       width=0
       color {colorUseRule=UseColor_Specified, specifiedColor=<rgba 0 0 0 0>}
       cornerRadii=<0 0 0 0>
      }
      tintColor {colorUseRule=UseColor_Specified, specifiedColor=<rgba 1 1 1 1>}
      uVRegion {bIsValid=false, max=<0 0>, min=<0 0>}
     }
     verticalTopSlotImage {
      bIsDynamicallyLoaded=false
      drawAs=NoDrawType
      imageType=NoImage
      mirroring=NoMirror
      resourceName=None
      resourceObject=null
      tiling=NoTile
      imageSize=<0 0>
      margin=<ltrb 0 0 0 0>
      outlineSettings {
       bUseBrushTransparency=false
       roundingType=HalfHeightRadius
       width=0
       color {colorUseRule=UseColor_Specified, specifiedColor=<rgba 0 0 0 0>}
       cornerRadii=<0 0 0 0>
      }
      tintColor {colorUseRule=UseColor_Specified, specifiedColor=<rgba 1 1 1 1>}
      uVRegion {bIsValid=false, max=<0 0>, min=<0 0>}
     }
    }
    widgetStyle {
     barThickness=8
     bottomShadowBrush {
      bIsDynamicallyLoaded=false
      drawAs=Box
      imageType=FullColor
      mirroring=NoMirror
      resourceName=../../../Engine/Content/Slate/Common/ScrollBoxShadowBottom.png
      resourceObject=null
      tiling=NoTile
      imageSize=<16 8>
      margin=<ltrb 0.5 0 0.5 1>
      outlineSettings {
       bUseBrushTransparency=false
       roundingType=HalfHeightRadius
       width=0
       color {colorUseRule=UseColor_Specified, specifiedColor=<rgba 0 0 0 0>}
       cornerRadii=<0 0 0 0>
      }
      tintColor {colorUseRule=UseColor_Specified, specifiedColor=<rgba 1 1 1 1>}
      uVRegion {bIsValid=false, max=<0 0>, min=<0 0>}
     }
     horizontalScrolledContentPadding=<ltrb 0 0 1 0>
     leftShadowBrush {
      bIsDynamicallyLoaded=false
      drawAs=Box
      imageType=FullColor
      mirroring=NoMirror
      resourceName=../../../Engine/Content/Slate/Common/ScrollBoxShadowLeft.png
      resourceObject=null
      tiling=NoTile
      imageSize=<8 16>
      margin=<ltrb 1 0.5 0 0.5>
      outlineSettings {
       bUseBrushTransparency=false
       roundingType=HalfHeightRadius
       width=0
       color {colorUseRule=UseColor_Specified, specifiedColor=<rgba 0 0 0 0>}
       cornerRadii=<0 0 0 0>
      }
      tintColor {colorUseRule=UseColor_Specified, specifiedColor=<rgba 1 1 1 1>}
      uVRegion {bIsValid=false, max=<0 0>, min=<0 0>}
     }
     rightShadowBrush {
      bIsDynamicallyLoaded=false
      drawAs=Box
      imageType=FullColor
      mirroring=NoMirror
      resourceName=../../../Engine/Content/Slate/Common/ScrollBoxShadowRight.png
      resourceObject=null
      tiling=NoTile
      imageSize=<8 16>
      margin=<ltrb 0 0.5 1 0.5>
      outlineSettings {
       bUseBrushTransparency=false
       roundingType=HalfHeightRadius
       width=0
       color {colorUseRule=UseColor_Specified, specifiedColor=<rgba 0 0 0 0>}
       cornerRadii=<0 0 0 0>
      }
      tintColor {colorUseRule=UseColor_Specified, specifiedColor=<rgba 1 1 1 1>}
      uVRegion {bIsValid=false, max=<0 0>, min=<0 0>}
     }
     topShadowBrush {
      bIsDynamicallyLoaded=false
      drawAs=Box
      imageType=FullColor
      mirroring=NoMirror
      resourceName=../../../Engine/Content/Slate/Common/ScrollBoxShadowTop.png
      resourceObject=null
      tiling=NoTile
      imageSize=<16 8>
      margin=<ltrb 0.5 1 0.5 0>
      outlineSettings {
       bUseBrushTransparency=false
       roundingType=HalfHeightRadius
       width=0
       color {colorUseRule=UseColor_Specified, specifiedColor=<rgba 0 0 0 0>}
       cornerRadii=<0 0 0 0>
      }
      tintColor {colorUseRule=UseColor_Specified, specifiedColor=<rgba 1 1 1 1>}
      uVRegion {bIsValid=false, max=<0 0>, min=<0 0>}
     }
     verticalScrolledContentPadding=<ltrb 0 0 0 1>
    }
   }
   slot {class=OverlaySlot, properties {horizontalAlignment=HAlign_Fill, padding=<ltrb 0 32 0 0>}}
  }
  VerticalBox_59 {
   widgetClassPath=&VerticalBox
   parent=CommonHierarchicalScrollBox_76
   widget=VerticalBox_59
   slot {class=ScrollBoxSlot, properties {verticalAlignment=VAlign_Top}}
  }
  Messages {
   widgetClassPath=&TextBlock
   bIsVariable=true
   parent=VerticalBox_59
   widget=Messages
   properties {
    autoWrapText=true
    text="Waiting for Pickaxe ID file input . . . . "
    visibility=Collapsed
    font {
     bForceMonospaced=false
     bMaterialIsStencil=false
     fontMaterial=null
     fontObject=&Roboto
     letterSpacing=0
     monospacedWidth=1
     size=12
     skewAmount=0
     typefaceFontName=Regular
     outlineSettings {
      bApplyOutlineToDropShadows=false
      bMiteredCorners=false
      bSeparateFillAlpha=false
      outlineMaterial=null
      outlineSize=0
      outlineColor=<rgba 0 0 0 1>
     }
    }
   }
   slot {class=VerticalBoxSlot}
  }
  Spacer_1 {
   widgetClassPath=&Spacer
   parent=VerticalBox_59
   widget=Spacer_1
   slot {class=VerticalBoxSlot, properties {padding=<ltrb 5 5 5 5>}}
  }
  Warnings {
   widgetClassPath=&TextBlock
   bIsVariable=true
   parent=VerticalBox_59
   widget=Warnings
   properties {
    autoWrapText=true
    text=Warnings
    visibility=Collapsed
    colorAndOpacity {colorUseRule=UseColor_Specified, specifiedColor=<rgba 1 0.6128 0 1>}
    font {
     bForceMonospaced=false
     bMaterialIsStencil=false
     fontMaterial=null
     fontObject=&Roboto
     letterSpacing=0
     monospacedWidth=1
     size=12
     skewAmount=0
     typefaceFontName=Regular
     outlineSettings {
      bApplyOutlineToDropShadows=false
      bMiteredCorners=false
      bSeparateFillAlpha=false
      outlineMaterial=null
      outlineSize=0
      outlineColor=<rgba 0 0 0 1>
     }
    }
   }
   slot {class=VerticalBoxSlot}
  }
  Spacer_2 {
   widgetClassPath=&Spacer
   parent=VerticalBox_59
   widget=Spacer_2
   slot {class=VerticalBoxSlot, properties {padding=<ltrb 5 5 5 5>}}
  }
  Errors {
   widgetClassPath=&TextBlock
   bIsVariable=true
   parent=VerticalBox_59
   widget=Errors
   properties {
    autoWrapText=true
    text=Errors
    visibility=Collapsed
    colorAndOpacity {colorUseRule=UseColor_Specified, specifiedColor=<rgba 1 0.0677 0.0677 1>}
    font {
     bForceMonospaced=false
     bMaterialIsStencil=false
     fontMaterial=null
     fontObject=&Roboto
     letterSpacing=0
     monospacedWidth=1
     size=12
     skewAmount=0
     typefaceFontName=Regular
     outlineSettings {
      bApplyOutlineToDropShadows=false
      bMiteredCorners=false
      bSeparateFillAlpha=false
      outlineMaterial=null
      outlineSize=0
      outlineColor=<rgba 0 0 0 1>
     }
    }
   }
   slot {class=VerticalBoxSlot}
  }
  Initial_PickaxeSelection {
   widgetClassPath=&DetailsView
   parent=CommonHierarchicalScrollBox_76
   widget=Initial_PickaxeSelection
   properties {categoriesToShow [Select Pickaxe to Edit]}
   slot {class=ScrollBoxSlot}
  }
  EUW_WID_FX_WID_Setup {
   widgetClassPath=&/EUW_WID_FX_WID_Setup
   bIsVariable=true
   parent=CommonHierarchicalScrollBox_76
   widget=EUW_WID_FX_WID_Setup
   slot {class=ScrollBoxSlot}
  }
  EUW_WID_FX_DualWieldWID_Setup {
   widgetClassPath=&/EUW_WID_FX_DualWieldWID_Setup
   bIsVariable=true
   parent=CommonHierarchicalScrollBox_76
   widget=EUW_WID_FX_DualWieldWID_Setup
   slot {class=ScrollBoxSlot}
  }
 ]
}
